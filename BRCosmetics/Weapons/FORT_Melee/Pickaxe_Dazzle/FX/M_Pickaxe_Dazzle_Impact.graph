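# Unlit translucent material for Niagara sprite and mesh particles, used for the Dazzle pickaxe impact effect.
# The impact texture's red channel is multiplied by particle alpha to form the base mask.
# Its green channel drives a smoothstep erosion, with the threshold set by the Erode dynamic parameter and the edge width by ErodeSmoothness.
# An ErodeAmount scalar is present but unconnected.
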Material /BRCosmetics/Weapons/FORT_Melee/Pickaxe_Dazzle/FX/M_Pickaxe_Dazzle_Impact
graph {
 properties {
  bUsedWithNiagaraMeshParticles=true
  bUsedWithNiagaraSprites=true
  blendMode=BLEND_Translucent
  pixelDepthOffsetMode=PDOM_Legacy
  shadingModel=MSM_Unlit
  usedShadingModels=Unlit
 }
 nodes [
  n1 = TextureSample() {texture=&/T_Pickaxe_Dazzle_Impacts}
  n2 = DynamicParameter() {defaultValue=<rgba -1 1 1 1>, paramNames [Erode,Param2,Param3,Param4]}
  n3 = ScalarParameter() {defaultValue=1, parameterName=ErodeSmoothness}
  Multiply(A=Multiply(A=@n1.R, B=ParticleColor().A), B=SmoothStep(Min=@n2.Erode, Max=Add(A=@n2.Erode, B=@n3), Value=@n1.G))
  ScalarParameter() {defaultValue=-1, parameterName=ErodeAmount}
 ]
}
