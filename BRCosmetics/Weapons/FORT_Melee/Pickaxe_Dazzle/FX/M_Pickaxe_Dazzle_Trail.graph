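# Translucent, unlit, two-sided material for the Dazzle pickaxe trail effect, usable on Niagara ribbons and mesh particles.
# It builds a tapered streak procedurally from UV math and wave distortion, with inner and edge masks plus panning-noise erosion.
# A smoothstepped smoke texture blends particle color toward a multiplied tint for emissive output.
# Three dynamic parameter sets expose per-particle shape, wave, time offset, erosion and overlay controls.
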Material /BRCosmetics/Weapons/FORT_Melee/Pickaxe_Dazzle/FX/M_Pickaxe_Dazzle_Trail
graph {
 properties {
  bUsedWithNiagaraMeshParticles=true
  bUsedWithNiagaraRibbons=true
  blendMode=BLEND_Translucent
  pixelDepthOffsetMode=PDOM_Legacy
  shadingModel=MSM_Unlit
  twoSided=true
  usedShadingModels=Unlit
 }
 nodes [
  n1 = ParticleColor()
  n2 = DynamicParameter() {parameterIndex=2, defaultValue=<rgba 1 1 0.4 1>, paramNames [TextureOverlaySmoothstepMin, TextureOverlaySmoothstepMax, ColorMult, Param4]}
  n3 = TextureCoordinate()
  n4 = TextureCoordinate() {coordinateIndex=1}
  n5 = ScalarParameter() {parameterName=TextureCoordforOverlayTexture}
  n6 = VectorParameter() {parameterName=Param_2, defaultValue=<rgba 1 0.2 0 1>}
  n7 = Add(
   A=Multiply(
    A=LinearInterpolate(A=@n3, B=@n4, Alpha=@n5)
    B=MaterialFunctionCall(X=@n6.R, Y=@n6.R) {materialFunction=&MakeFloat2}.Result
   )
   B=MaterialFunctionCall(X=Constant(), Y=NamedRerouteUsage()) {materialFunction=&MakeFloat2}.Result
  )
  n8 = TextureObjectParameter() {parameterName=ColorOverlayTexture, texture=&T_Inky_Smoke_Tile}
  n9 = OneMinus(None=SmoothStep(
    Min=@n2.TextureOverlaySmoothstepMin
    Max=@n2.TextureOverlaySmoothstepMin
    Value=TextureSample(UVs=@n7, Tex=@n8) {texture=&T_Inky_Smoke_Tile}.R
   ))
  n10 = LinearInterpolate(A=OneMinus(None=NamedRerouteUsage())) {constAlpha=0.65}
  MaterialFunctionCall(Emissive=LinearInterpolate(A=@n1.RGB, B=Multiply(A=@n1.RGB, B=@n2.ColorMult) {constB=0}, Alpha=Multiply(A=@n9, B=@n10))) {materialFunction=&MF_FX_EmissiveAdjustments}
  n11 = OneMinus(None=Time())
  n12 = ScalarParameter() {defaultValue=0.25, parameterName=TimeMult}
  n13 = DynamicParameter() {defaultValue=<rgba 0.5 2 0 1.1>, paramNames [TimeOffsetRandom, WaveFrequency, TrailOffset, TrailWidth]}
  NamedRerouteDeclaration(None=Add(A=Multiply(A=@n11, B=@n12) {constB=2}, B=@n13.TimeOffsetRandom)) {name=RandomTime, nodeColor=<rgba 1 0.0469 0 1>}
  NamedRerouteDeclaration(None=@n13.TrailWidth) {name=TrailWidth, nodeColor=<rgba 1 0 0.8906 1>}
  n14 = DynamicParameter() {parameterIndex=1, defaultValue=<rgba 1 0.1 1 1>, paramNames [UV Offset, Wave Amp, TrailLen, Erode]}
  NamedRerouteDeclaration(None=@n14.TrailLen) {name=TrailLen, nodeColor=<rgba 1 0.8672 0 1>}
  NamedRerouteDeclaration(None=@n14.Erode) {name=Erode, nodeColor=<rgba 0 1 0.7891 1>}
  n15 = NamedRerouteUsage()
  n16 = Constant()
  n17 = SmoothStep(Min=@n16, Max=Add(A=@n16, B=NamedRerouteUsage()), Value=ComponentMask(None=TextureCoordinate()) {r=true})
  n18 = Constant()
  n19 = Constant() {r=0.5}
  n20 = Add(A=SmoothStep(Min=@n18, Max=Add(A=@n19, B=@n18), Value=@n17), B=Constant() {r=1}) {constB=0}
  n21 = Clamp(None=Min(A=OneMinus(None=@n17), B=Saturate(None=@n20)))
  n22 = Divide(
   A=ConstantBiasScale(None=ComponentMask(None=@n15) {r=true}) {bias=-0.5, scale=2}
   B=Multiply(A=Multiply(A=@n21, B=@n21), B=NamedRerouteUsage())
  )
  n23 = OneMinus(None=Saturate(None=DotProduct(A=@n22, B=@n22)))
  n24 = OneMinus(None=Step(Y=@n23) {constX=0})
  n25 = Multiply(A=Multiply(A=@n24, B=NamedRerouteUsage()), B=NamedRerouteUsage())
  n26 = NamedRerouteUsage()
  NamedRerouteDeclaration(None=Multiply(A=Multiply(A=@n25, B=@n26), B=ScalarParameter() {defaultValue=0.9, parameterName=Erosion})) {name=Base, nodeColor=<rgba 1 0.2831 0 1>}
  n27 = ScalarParameter() {parameterName=LeadingEdgeMaskAmount}
  n28 = ScalarParameter() {defaultValue=0.1, parameterName=LeadingEdgeMaskSmoothness}
  n29 = NamedRerouteUsage()
  NamedRerouteDeclaration(None=SmoothStep(Min=@n27, Max=Add(A=@n27, B=@n28), Value=ComponentMask(None=@n29) {g=true})) {name=LeadingEdgeMask, nodeColor=<rgba 0 0.8824 1 1>}
  n30 = Multiply(A=OneMinus(None=@n23), B=@n24)
  n31 = OneMinus(None=@n30)
  n32 = ScalarParameter() {defaultValue=4, parameterName=InnerControl}
  n33 = ScalarParameter() {defaultValue=5, parameterName=EdgeControl}
  n34 = Add(
   A=Multiply(A=Power(Base=@n31, Exp=@n32), B=ScalarParameter() {defaultValue=0.5, parameterName=InnerMult})
   B=Multiply(A=Power(Base=@n30, Exp=@n33), B=ScalarParameter() {defaultValue=1, parameterName=EdgeMult})
  )
  n35 = Saturate(None=LinearInterpolate(A=Saturate(None=@n34), Alpha=ScalarParameter() {defaultValue=0.2, parameterName=Blend}))
  n36 = NamedRerouteUsage()
  NamedRerouteDeclaration(None=Saturate(None=Add(A=@n35, B=@n36))) {name=InnerDetails, nodeColor=<rgba 0.2423 0.1413 1 1>}
  n37 = ScalarParameter() {defaultValue=0.4, parameterName=InnerDetailsMaskAmount}
  n38 = ScalarParameter() {defaultValue=0.2, parameterName=InnerDetailsMaskSmoothness}
  n39 = ComponentMask(None=NamedRerouteUsage()) {g=true}
  NamedRerouteDeclaration(None=SmoothStep(Min=@n37, Max=Add(A=@n37, B=@n38), Value=@n39)) {name=InnerDetailsMask, nodeColor=<rgba 1 0 0.2086 1>}
  n40 = VectorParameter() {parameterName=Param_1, defaultValue=<rgba 1 0.1 0 1>}
  n41 = Multiply(
   A=TextureCoordinate()
   B=MaterialFunctionCall(X=@n40.R, Y=@n40.R) {materialFunction=&MakeFloat2}.Result
  )
  n42 = ConstantBiasScale(None=@n22)
  n43 = Add(A=Panner(Coordinate=Add(A=@n41, B=@n42) {constB=3}) {speedX=2}, B=AppendVector(A=Constant(), B=NamedRerouteUsage()))
  n44 = TextureObjectParameter() {parameterName=ErosionTexture, texture=/Game/Effects/Fort_Effects/Textures/Tile/T_GaussianNoise_01_M}
  n45 = Multiply(
   A=Power(
    Base=TextureSample(UVs=@n43, Tex=@n44) {texture=&T_GaussianNoise_01_M}.R
    Exp=ScalarParameter() {defaultValue=1, parameterName=LTrailContrast}
   ) {constExponent=1.25}
   B=ScalarParameter() {defaultValue=1, parameterName=LTrailStrenght}
  ) {constB=2}
  NamedRerouteDeclaration(None=@n45) {name=Waves, nodeColor=<rgba 1 0 0.6328 1>}
  NamedRerouteDeclaration(None=@n13.WaveFrequency) {name=WaveFrequency, nodeColor=<rgba 0.4766 1 0 1>}
  NamedRerouteDeclaration(None=@n14.Wave Amp) {name=WaveAmp, nodeColor=<rgba 1 0.2578 0 1>}
  NamedRerouteDeclaration(None=@n13.TrailOffset) {name=TrailOffset, nodeColor=<rgba 0.5703 0 1 1>}
  n46 = TextureCoordinate()
  NamedRerouteDeclaration(None=AppendVector(A=ComponentMask(None=@n46) {g=true}, B=ComponentMask(None=@n46) {r=true})) {name=UV, nodeColor=<rgba 1 0.7808 0 1>}
  NamedRerouteDeclaration(None=@n14.UV Offset) {name=UV Offset, nodeColor=<rgba 0 0.7187 1 1>}
  n47 = NamedRerouteUsage()
  n48 = AppendVector(A=Constant(), B=NamedRerouteUsage())
  n49 = Constant()
  NamedRerouteDeclaration(None=MaterialFunctionCall!(
    UV=Add(A=@n47, B=@n48)
    U Frequency
    V Frequency=NamedRerouteUsage()
    U Amplitude
    V Amplitude=ScalarParameter() {defaultValue=0.18, parameterName=WaveAmp}
    U Speed
    V Speed
    U Time=@n49
    V Time=@n49
   ) {materialFunction=/Game/Effects/Fort_Effects/MaterialFunctions/MF_WaveUV}.Wave UVs) {name=WaveUV, nodeColor=<rgba 1 0.6094 0 1>}
  NamedRerouteUsage()
  OneMinus(None=@n39)
  n50 = NamedRerouteUsage()
  n51 = ScalarParameter() {parameterName=Smoothness}
  n52 = OneMinus(None=SmoothStep(Min=@n50, Max=Add(A=@n50, B=@n51), Value=Saturate(None=@n45)))
  Saturate(None=Multiply(A=Saturate(None=@n52), B=ParticleColor().A))
  ScalarParameter() {defaultValue=1.1, parameterName=Trail Thickness L}
 ]
}
