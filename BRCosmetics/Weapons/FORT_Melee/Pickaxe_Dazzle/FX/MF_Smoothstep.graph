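# Material function MF_Smoothstep, a scalar smoothstep helper in the Pickaxe_Dazzle FX folder.
# It takes three scalar inputs: Input, Amount and Smoothness.
# Amount sets the lower edge and Amount plus Smoothness sets the upper edge of a SmoothStep applied to Input, which is returned as the single output.
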MaterialFunction /BRCosmetics/Weapons/FORT_Melee/Pickaxe_Dazzle/FX/MF_Smoothstep
graph {nodes [
  n1 = FunctionInput() {inputName=Amount, inputType=FunctionInput_Scalar, sortPriority=1}
  n2 = FunctionInput() {inputName=Smoothness, inputType=FunctionInput_Scalar, sortPriority=2}
  FunctionOutput(None=SmoothStep(Min=@n1, Max=Add(A=@n1, B=@n2), Value=FunctionInput() {inputName=Input, inputType=FunctionInput_Scalar}))
 ]}
