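# Translucent, unlit, two-sided Niagara material for the underlay of the Dazzle pickaxe trail effect.
# Emissive blends particle color toward a ColorMult-scaled shade, masked by inverted inky smoke and panning Gaussian detail noise.
# A smoothstep erodes Gaussian noise against a soft UV-based mask scaled by the Erode dynamic parameter.
# Per-particle random offsets vary the texture sampling between particles.
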Material /BRCosmetics/Weapons/FORT_Melee/Pickaxe_Dazzle/FX/M_Pickaxe_Dazzle_Trail_Underlay
graph {
 properties {
  bUsedWithNiagaraMeshParticles=true
  bUsedWithNiagaraRibbons=true
  bUsedWithNiagaraSprites=true
  blendMode=BLEND_Translucent
  pixelDepthOffsetMode=PDOM_Legacy
  shadingModel=MSM_Unlit
  twoSided=true
  usedShadingModels=Unlit
 }
 nodes [
  n1 = ParticleColor()
  n2 = ScalarParameter() {defaultValue=0.3, parameterName=ColorMult}
  n3 = VectorParameter() {parameterName=UV_Scale, defaultValue=<rgba 0.2 1 0 1>}
  n4 = Multiply(
   A=TextureCoordinate()
   B=MaterialFunctionCall(X=@n3.R, Y=@n3.R) {materialFunction=&MakeFloat2}.Result
  )
  n5 = MaterialFunctionCall(X=Constant(), Y=ParticleRandom()) {materialFunction=&MakeFloat2}
  n6 = OneMinus(None=TextureSample(UVs=Add(A=@n4, B=@n5.Result)) {texture=&T_Inky_Smoke_Tile}.R)
  n7 = LinearInterpolate(A=OneMinus(None=NamedRerouteUsage())) {constAlpha=0.65}
  MaterialFunctionCall(Emissive=LinearInterpolate(A=@n1.RGB, B=Multiply(A=@n1.RGB, B=@n2) {constB=0}, Alpha=Multiply(A=@n6, B=@n7))) {materialFunction=&MF_FX_EmissiveAdjustments}
  n8 = TextureCoordinate()
  n9 = ParticleRandom()
  n10 = MaterialFunctionCall(Float2=TextureCoordinate()) {materialFunction=&BreakOutFloat2Components}
  n11 = MaterialFunctionCall(X=@n10.R, Y=@n10.R) {materialFunction=&MakeFloat2}
  n12 = Multiply(
   A=SphereMask(A=@n11.Result, B=Constant() {r=0.5}, Hardness=Constant() {r=0.8}) {attenuationRadius=0.5, hardnessPercent=0}
   B=MaterialFunctionCall!(
    Input=MaterialFunctionCall(Float2=@n11.Result) {materialFunction=&BreakOutFloat2Components}.R
    Amount=Constant()
    Smoothness=Constant() {r=0.3}
   ) {materialFunction=&/MF_Smoothstep}.Result
  )
  n13 = DynamicParameter() {defaultValue=<rgba 1 2 0 1.1>, paramNames [Erode, "", "", ""]}
  MaterialFunctionCall(
   Input=TextureSample(UVs=Add(A=@n8, B=@n9)) {texture=/Game/Effects/Fort_Effects/Textures/Tile/T_GaussianNoise_01_M}.R
   Amount=OneMinus(None=Multiply(A=@n12, B=@n13.Erode))
   Smoothness=Constant()
  ) {materialFunction=&MF_Smoothstep}
  n14 = VectorParameter() {parameterName=UV_Scale_Details, defaultValue=<rgba 0.2 1 0 1>}
  n15 = Panner(Coordinate=Multiply(
    A=TextureCoordinate()
    B=MaterialFunctionCall(X=@n14.R, Y=@n14.R) {materialFunction=&MakeFloat2}.Result
   )) {speedX=0.2}
  n16 = AppendVector(A=Constant(), B=ParticleRandom())
  NamedRerouteDeclaration(None=TextureSample(UVs=Add(A=@n15, B=@n16)) {texture=&T_GaussianNoise_01_M}.R) {name=Details, nodeColor=<rgba 1 0 0.6328 1>}
 ]
}
