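# Niagara impact effect for the Dazzle pickaxe, bursting toon-shaded, outlined meshes (bunny, emote, plus, stars) alongside flares, glow drops, outline sprites, smoke and sparks.
# Mesh shapes fly out with drag and curl noise, with position and scale jittered in time-stepped offsets per spawn group.
# User color arrays and switch parameters select color variants, including alternate impact flare emitters.
# Front-end modifiers shrink it for previews, and time-of-day scalars adjust glow brightness.
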NiagaraSystem /BRCosmetics/Weapons/FORT_Melee/Pickaxe_Dazzle/FX/NS_Pickaxe_Dazzle_Impact
dependencies {
 usedDataInterfaces [
  &NiagaraDataInterfaceCurve
  &NiagaraDataInterfaceVector2DCurve
  &NiagaraDataInterfaceColorCurve
  &NiagaraDataInterfaceArrayFloat
  &NiagaraDataInterfaceArrayFloat3
 ]
 usedDynamicInputs [&ReturnExecIndex, &FN_Select_Float_from_Time_Of_Day]
 usedModules [
  &SystemState
  &self:SystemSpawnScript.SystemScriptSource.SystemScriptGraph.NiagaraNodeAssignment_0.SetVariables_#1
 ]
}
emitters {
 Bunny {
  emitterUpdateScript [
   &EmitterState()
   &SpawnBurst_Instantaneous(
    Spawn Count=&RandomRangeInt(Minimum=1, Maximum=2)
    Spawn Group=1
    Use Spawn Probability=false
    UseLoopCountLimit=false
   )
   /Game/Effects/Niagara/MasterMaterials/ToonShader/Template/NM_ToonShaderParameters(Lighting_Hardness=1, RimLight_Color:LinearColor=<rgba 1 1 1 1>, RimLight_Width=0)
   /Game/Effects/Niagara/MasterMaterials/ToonShader/Template/NM_OutlineParameters(OutlineWidth=0.7)
  ]
  particleSpawnScript [
   &PartitionParticles()
   &FN_SetSpawnGroup(Number of Spawn Groups=6)
   &InitializeParticle(
    Color=&LerpLinearColors(
     StartColor=&SelectLinearColorFromArray(Array Sampling Mode=::NewEnumerator0, Direct Array Index=?, Color Selection Array=$User.MeshColors, Random Seed=0)
     EndColor=&SelectLinearColorFromArray(Array Sampling Mode=::NewEnumerator0, Direct Array Index=?, Color Selection Array=$User.MeshColors_Variant_01, Random Seed=0)
     LerpFactor=$User.ColorSwitcher
    )
    Lifetime Max=0.5
    Lifetime Min=0.25
    Lifetime Mode=::NewEnumerator1
    Mesh Scale Mode=::NewEnumerator1
    Mesh Uniform Scale Max=3.5
    Mesh Uniform Scale Min=2
    Position Offset Coordinate Space=::Local
    Sprite UV Mode=::NewEnumerator3
    UsePositionOffset=true
    Write Material Random=true
   ) {module=InitializeParticle001}
   &MeshRotationForce(Rotation=&VectorFromFloat(Value=1))
   &InitialMeshOrientation(Use Rotation Vector=false) {module=InitialMeshOrientation001}
   &AddVelocity(
    Velocity=&RandomRangeVector(Minimum=<200 -1300 -1300>, Maximum=<600 1300 1300>)
    Velocity Speed Scale=&RandomRangeFloat(Minimum=0.3, Maximum=0.8)
   ) {module=AddVelocity001}
   &ApplyInitialForces()
   &CameraOffset(Camera Offset Amount=25)
   /SetVariables(Particles.Color=&MultiplyLinearColor(
     Linear Color=$Particles.Color
     Scale Factor=&MakeFloat4FromVectorAndFloat("Vector (RGB)"=&VectorFromFloat(Value=1), "Float (Alpha)"=1)
    )) {isSetParameters=true}
   /SetVariables(Particles.Scale=&Multiply_VectorByFloat(Vector=$Particles.Scale, Float=&FN_FrontEndFloatModifier(Gameplay=1, Front End=0.3))) {isSetParameters=true}
  ]
  particleUpdateScript [
   &ParticleState()
   &Drag(Drag=&RandomRangeFloat(Minimum=8, Maximum=15), Use Linear Drag=true, Use Rotational Drag=false) {module=Drag001}
   &SpriteRotationRate(Rotation Rate=&RandomRangeFloat(Minimum=50, Maximum=170))
   &GravityForce(Gravity=&RandomRangeVector(Minimum=<0 0 150>, Maximum=<0 0 -150>))
   &CurlNoiseForce(Pan Noise=false)
   &ScaleMeshSize(Scale Factor=&Multiply_VectorByFloat(
     Vector=<1 1 1>
     Float=&FloatFromCurve(FloatCurve=DataInterface() {propertyValues {Curve {keys [
          {interpMode=RCIM_Cubic, tangentMode=RCTM_User, time=0, value=0}
          <at 0.2 1>
          {interpMode=RCIM_Cubic, tangentMode=RCTM_User, time=0.8, value=1}
          {tangentMode=RCTM_User, time=1, value=0}
         ]}}})
    )) {module=ScaleMeshSize006}
   &ScaleSpriteSize(Uniform Curve Sprite Scale=DataInterface() {propertyValues {Curve {keys [
        {interpMode=RCIM_Cubic, tangentMode=RCTM_User, time=0, value=0}
        <at 0.2 1>
        {interpMode=RCIM_Cubic, tangentMode=RCTM_User, time=0.7, value=1}
        {arriveTangent=-0.9819, interpMode=RCIM_Cubic, leaveTangent=-0.9819, tangentMode=RCTM_User, time=0.85, value=1.3}
        {arriveTangent=-9.5654, interpMode=RCIM_Cubic, leaveTangent=-9.5654, tangentMode=RCTM_User, time=1, value=0}
       ]}}})
   &SolveForcesAndVelocity(Clamp Velocity=false, Limit Acceleration=false)
   /SetVariables(
    Particles.Position=&AddVectorToPosition(
     Position=$Particles.Position
     Vector=&Multiply_Vector(
      A=&VectorFromFloat(Value=&FN_FrontEndFloatModifier(Gameplay=0.8, Front End=0.1))
      B=&SelectVectorFromArray(
       Direct Array Index=&ModuloInt(
        A=&MakeIntFromFloat(Float=&Add_Float(
          A=&Multiply_Float(A=$Engine.Time, B=6)
          B=&MakeFloatFromInt(Int=$Particles.SpawnGroup)
         ))
        B=6
       )
       Vector Selection Array=DataInterface() {propertyValues {FloatData [<1 0 0>, <0 -0.5 -0.5>, <-1 0.5 0.5>, <0.5 2 1>, <0.1 -0.4 0.2>, <0.35 0.7 -0.4>]}}
      )
     )
    )
    Particles.Scale=&Multiply_Vector(
     A=$Particles.Scale
     B=&SelectVectorFromArray(
      Direct Array Index=&ModuloInt(
       A=&MakeIntFromFloat(Float=&Add_Float(
         A=&Multiply_Float(A=$Engine.Time, B=6)
         B=&MakeFloatFromInt(Int=$Particles.SpawnGroup)
        ))
       B=3
      )
      Vector Selection Array=DataInterface() {propertyValues {FloatData [<1 1 1>, <1.05 1 1.05>, <0.95 1.02 0.99>]}}
     )
    )
   ) {isSetParameters=true}
  ]
  rendererClasses [&NiagaraMeshRendererProperties]
  renderers [
   {rendererClass=&NiagaraMeshRendererProperties, rendererIndex=0}
   {rendererClass=&NiagaraMeshRendererProperties, rendererIndex=1}
  ]
 }
 Emote {
  emitterUpdateScript [
   &EmitterState()
   &SpawnBurst_Instantaneous(Spawn Group=1, Use Spawn Probability=false, UseLoopCountLimit=false)
   &NM_ToonShaderParameters(Lighting_Hardness=1, RimLight_Color:LinearColor=<rgba 1 1 1 1>, RimLight_Width=0)
   &NM_OutlineParameters(OutlineWidth=0.25)
  ]
  particleSpawnScript [
   &PartitionParticles()
   &FN_SetSpawnGroup(Number of Spawn Groups=6)
   &InitializeParticle(
    Color=&LerpLinearColors(
     StartColor=&SelectLinearColorFromArray(Array Sampling Mode=::NewEnumerator0, Direct Array Index=?, Color Selection Array=$User.MeshColors, Random Seed=0)
     EndColor=&SelectLinearColorFromArray(Array Sampling Mode=::NewEnumerator0, Direct Array Index=?, Color Selection Array=$User.MeshColors_Variant_01, Random Seed=0)
     LerpFactor=$User.ColorSwitcher
    )
    Lifetime Max=0.5
    Lifetime Min=0.25
    Lifetime Mode=::NewEnumerator1
    Mesh Scale Mode=::NewEnumerator1
    Mesh Uniform Scale Max=3.5
    Mesh Uniform Scale Min=2
    Position Offset Coordinate Space=::Local
    Sprite UV Mode=::NewEnumerator3
    UsePositionOffset=true
    Write Material Random=true
   ) {module=InitializeParticle001}
   &MeshRotationForce(Rotation=&VectorFromFloat(Value=1))
   &InitialMeshOrientation(Use Rotation Vector=false) {module=InitialMeshOrientation001}
   &AddVelocity(
    Velocity=&RandomRangeVector(Minimum=<200 -1300 -1300>, Maximum=<600 1300 1300>)
    Velocity Speed Scale=&RandomRangeFloat(Minimum=0.3, Maximum=0.8)
   ) {module=AddVelocity001}
   &ApplyInitialForces()
   &CameraOffset(Camera Offset Amount=25)
   /SetVariables(Particles.Color=&MultiplyLinearColor(
     Linear Color=$Particles.Color
     Scale Factor=&MakeFloat4FromVectorAndFloat("Vector (RGB)"=&VectorFromFloat(Value=1), "Float (Alpha)"=1)
    )) {isSetParameters=true}
   /SetVariables(Particles.Scale=&Multiply_VectorByFloat(Vector=$Particles.Scale, Float=&FN_FrontEndFloatModifier(Gameplay=1, Front End=0.3))) {isSetParameters=true}
  ]
  particleUpdateScript [
   &ParticleState()
   &Drag(Drag=&RandomRangeFloat(Minimum=8, Maximum=15), Use Linear Drag=true, Use Rotational Drag=false) {module=Drag001}
   &SpriteRotationRate(Rotation Rate=&RandomRangeFloat(Minimum=50, Maximum=170))
   &GravityForce(Gravity=&RandomRangeVector(Minimum=<0 0 150>, Maximum=<0 0 -150>))
   &CurlNoiseForce(Pan Noise=false)
   &ScaleMeshSize(Scale Factor=&Multiply_VectorByFloat(
     Vector=<1 1 1>
     Float=&FloatFromCurve(FloatCurve=DataInterface() {propertyValues {Curve {keys [
          {interpMode=RCIM_Cubic, tangentMode=RCTM_User, time=0, value=0}
          <at 0.2 1>
          {interpMode=RCIM_Cubic, tangentMode=RCTM_User, time=0.8, value=1}
          {tangentMode=RCTM_User, time=1, value=0}
         ]}}})
    )) {module=ScaleMeshSize006}
   &ScaleSpriteSize(Uniform Curve Sprite Scale=DataInterface() {propertyValues {Curve {keys [
        {interpMode=RCIM_Cubic, tangentMode=RCTM_User, time=0, value=0}
        <at 0.2 1>
        {interpMode=RCIM_Cubic, tangentMode=RCTM_User, time=0.7, value=1}
        {arriveTangent=-0.9819, interpMode=RCIM_Cubic, leaveTangent=-0.9819, tangentMode=RCTM_User, time=0.85, value=1.3}
        {arriveTangent=-9.5654, interpMode=RCIM_Cubic, leaveTangent=-9.5654, tangentMode=RCTM_User, time=1, value=0}
       ]}}})
   &SolveForcesAndVelocity(Clamp Velocity=false, Limit Acceleration=false)
   /SetVariables(
    Particles.Position=&AddVectorToPosition(
     Position=$Particles.Position
     Vector=&Multiply_Vector(
      A=&VectorFromFloat(Value=&FN_FrontEndFloatModifier(Gameplay=0.8, Front End=0.1))
      B=&SelectVectorFromArray(
       Direct Array Index=&ModuloInt(
        A=&MakeIntFromFloat(Float=&Add_Float(
          A=&Multiply_Float(A=$Engine.Time, B=6)
          B=&MakeFloatFromInt(Int=$Particles.SpawnGroup)
         ))
        B=6
       )
       Vector Selection Array=DataInterface() {propertyValues {FloatData [<1 0 0>, <0 -0.5 -0.5>, <-1 0.5 0.5>, <0.5 2 1>, <0.1 -0.4 0.2>, <0.35 0.7 -0.4>]}}
      )
     )
    )
    Particles.Scale=&Multiply_Vector(
     A=$Particles.Scale
     B=&SelectVectorFromArray(
      Direct Array Index=&ModuloInt(
       A=&MakeIntFromFloat(Float=&Add_Float(
         A=&Multiply_Float(A=$Engine.Time, B=6)
         B=&MakeFloatFromInt(Int=$Particles.SpawnGroup)
        ))
       B=3
      )
      Vector Selection Array=DataInterface() {propertyValues {FloatData [<1 1 1>, <1.05 1 1.05>, <0.95 1.02 0.99>]}}
     )
    )
   ) {isSetParameters=true}
  ]
  rendererClasses [&NiagaraMeshRendererProperties]
  renderers [
   {rendererClass=&NiagaraMeshRendererProperties, rendererIndex=0}
   {rendererClass=&NiagaraMeshRendererProperties, rendererIndex=1}
  ]
 }
 Flare {
  emitterUpdateScript [
   &EmitterState(Life Cycle Mode=::NewEnumerator1, Loop Behavior=::NewEnumerator1, Loop Delay=0.1, Loop Duration=1, UseLoopDelay=false)
   &SpawnBurst_Instantaneous(Spawn Count=1, Use Spawn Probability=false)
  ]
  particleSpawnScript [
   &InitializeParticle(
    Color=<rgba 0 0 0 1>
    Lifetime=0.25
    Sprite Rotation Mode=::NewEnumerator2
    Sprite Size Mode=::NewEnumerator3
    Uniform Sprite Size=200
    UsePositionOffset=false
    Write Material Random=true
   )
   &DynamicMaterialParameters(
    Index 0 Param 1=&RandomRangeFloat(Minimum=0.4, Maximum=0.6)
    Index 0 Param 2=1
    Index 0 Param 3=1
    Index 0 Param 4=1
    Module.Param0WriteEnabled=true
    Module.Param1WriteEnabled=true
    Module.Param2WriteEnabled=true
    Module.Param3WriteEnabled=true
   )
   /SetVariables(Particles.Dark_Glow_CameraOffset=&MakeCustomFloatFromBool(Bool=$User.bIsFrontEndPreview, True Float=1, False Float=50)) {isSetParameters=true}
   &CameraOffset(Camera Offset Amount=&MakeCustomFloatFromBool(Bool=$User.bIsFrontEndPreview, True Float=1, False Float=30))
  ]
  particleUpdateScript [
   /SetVariables(Particles.Dark_Glow=&MakeLinearColorFromVectorAndFloat(
     "Vector (RGB)"=&MakeVectorFromLinearColorRGB(Color=&LerpLinearColors(
       StartColor=&MultiplyLinearColor(
        Linear Color=<rgba 0.0733 0.3164 0.3837 0.01>
        Scale Factor=&MakeVector4_FromFloat(Float=&FNTimeOfDayScalar(Night Scale=0.5, Morning Scale=0.5, Evening Scale=0.3))
       )
       EndColor=&MultiplyLinearColor(
        Linear Color=<rgba 0.3837 0.0733 0.2802 0.01>
        Scale Factor=&MakeVector4_FromFloat(Float=&FNTimeOfDayScalar(Night Scale=0.5, Morning Scale=0.5, Evening Scale=0.3))
       )
       LerpFactor=$User.VariantSwitch
      ))
     "Float (Alpha)"=&Multiply_Float(
      A=&Multiply_Float(A=&FNTimeOfDayScalar(Night Scale=0.5, Morning Scale=0.5, Evening Scale=1), B=0.5)
      B=&FloatFromCurve(FloatCurve=DataInterface() {propertyValues {Curve {keys [<at 0 0>,<at 0.2 1>,<at 0.7 0.6>,<at 1 0>]}}})
     )
    )) {isSetParameters=true}
   /SetVariables(Particles.Dark_Glow_Size=&Multiply_Vector2(
     A=&Vector2DFromFloat(Value=220)
     B=&Vector2DFromFloat(Value=&FloatFromCurve(
       FloatCurve=DataInterface() {propertyValues {Curve {keys [<at 0 1.5 RCIM_Cubic>,<at 1 0.5>]}}}
       Scale Curve=&FN_FrontEndFloatModifier(Gameplay=1, Front End=0.6)
      ))
    )) {isSetParameters=true}
   &ParticleState()
   &ScaleColor(
    Scale Alpha=&Multiply_Float(
     A=1
     B=&Multiply_Float(
      A=&FNTimeOfDayScalar(Night Scale=0.5, Morning Scale=0.5, Evening Scale=0.5)
      B=&FloatFromCurve(
       FloatCurve=DataInterface() {propertyValues {Curve {keys [<at 0 0>,<at 0.2 1>,<at 0.8 1>,<at 1 0>]}}}
       Scale Curve=&FNTimeOfDayScalar(Night Scale=0.5, Morning Scale=0.5, Evening Scale=0.5)
      )
     )
    )
    Scale RGB=<2 2 2>
    ScaleA=true
    ScaleRGB=true
   )
   &ScaleSpriteSize(Scale Factor=&Vector2DFromFloat(Value=&FloatFromCurve(
      FloatCurve=DataInterface() {propertyValues {Curve {keys [<at 0 1.5 RCIM_Cubic>,<at 1 0>]}}}
      Scale Curve=&FN_FrontEndFloatModifier(Gameplay=1, Front End=0.6)
     )))
  ]
  rendererClasses [&NiagaraSpriteRendererProperties]
  renderers [
   {rendererClass=&NiagaraSpriteRendererProperties}
   {rendererClass=&NiagaraSpriteRendererProperties, rendererIndex=1}
  ]
 }
 GlowDrops {
  emitterUpdateScript [
   &EmitterState(
    Enable Distance Culling=true
    Max Distance Response=::KillImmediately
    Module.bApplySpawnCountScaleByDistance=true
    Module.bUseMaxDistance=true
    Module.bUseMinDistance=false
    Scalability Mode=::NewEnumerator1
    Scale Spawn Count=true
    Spawn Count Scale By Distance=DataInterface() {propertyValues {Curve {keys [<at 0.3676 0.9657>,<at 0.9962 0.4982>]}}}
   )
   &SpawnBurst_Instantaneous(Spawn Count=&UniformRangedInt(Minimum=7, Maximum=12), Use Spawn Probability=false)
  ]
  particleSpawnScript [
   /Niagara/Modules/Spawn/Initialization/InitializeParticle(
    Color=&LerpLinearColors(StartColor=<rgba 0.95 0 1 1>, EndColor=<rgba 0 0.0583 0.5 1>, LerpFactor=$User.ColorSwitcher)
    Lifetime=&UniformRangedFloat(Minimum=0.25, Maximum=0.35)
    Module.Write Color=true
    Module.Write Lifetime=true
    Module.Write Mass=true
    Module.Write Material Random=true
    Module.Write Position=false
    Module.Write Scale=false
    Module.Write SpriteRotation=false
    Module.Write SpriteSize=true
    Position=&SimulationPosition()
    Sprite Size=&Vector2DFromFloat(Value=&MakeFloatFromVector(Vector=&UniformRangedVector(Minimum=<2.5 2.5 2.5>, Maximum=<4 4 4>), Channel=::NewEnumerator0))
   )
   /Niagara/Modules/Spawn/Location/SphereLocation(MaskSpawnGroup=false, Module.Override Randomness=false, Sphere Radius=5, Surface Only=false)
   &AddVelocityFromPoint(
    Module.Use Falloff Distance=false
    Velocity Strength=&Multiply_Float(
     A=&UniformRangedFloat(Minimum=350, Maximum=750)
     B=&FN_FrontEndFloatModifier(Gameplay=1, Front End=0.8)
    )
   )
   /SetVariables(Particles.SpriteSize=&Multiply_Vector2D_ByFloat(Float=&FN_FrontEndFloatModifier(Gameplay=1, Front End=0.5), Vector2D=$Particles.SpriteSize)) {isSetParameters=true}
  ]
  particleUpdateScript [
   &ParticleState()
   &GravityForce(Gravity=<0 0 -250>)
   &Drag(Drag=&UniformRangedFloat(Minimum=2, Maximum=6), Use Linear Drag=true, Use Rotational Drag=false)
   &SolveForcesAndVelocity(Clamp Velocity=false, Limit Acceleration=false)
   &ScaleSpriteSize(Scale Factor=&Vector2DFromCurve(
     Vector2Curve=DataInterface() {propertyValues {
       XCurve {keys [
         {arriveTangent=5.9728, interpMode=RCIM_Cubic, leaveTangent=5.9728, tangentMode=RCTM_User, time=0, value=0.1}
         {arriveTangent=-0.1, interpMode=RCIM_Cubic, leaveTangent=-0.1, tangentMode=RCTM_User, time=0.18, value=0.8}
         {interpMode=RCIM_Cubic, tangentMode=RCTM_User, time=1, value=0}
        ]}
       YCurve {keys [
         {arriveTangent=6.2715, interpMode=RCIM_Cubic, leaveTangent=6.2715, tangentMode=RCTM_User, time=0, value=0.1}
         {arriveTangent=-0.1001, interpMode=RCIM_Cubic, leaveTangent=-0.1001, tangentMode=RCTM_User, time=0.18, value=0.8}
         {interpMode=RCIM_Cubic, tangentMode=RCTM_User, time=0.9988, value=0}
        ]}
      }}
     ScaleCurve=<2 2>
    ))
   &ScaleSpriteSizeBySpeed(Max Scale Factor=<0.5 8>, Min Scale Factor=<0.8 1>)
   &ScaleColor(
    Scale Alpha=&FloatFromCurve(FloatCurve=DataInterface() {propertyValues {Curve {keys [<at 0.4 1.0113 RCIM_Cubic>,<at 1.0011 0.002>]}}})
    Scale RGB=&Multiply_VectorByFloat(
     Vector=&MakeVectorFromLinearColorRGB(Color=&ColorFromCurve(ColorCurve=DataInterface() {propertyValues {
         AlphaCurve {keys []}
         BlueCurve {keys [<at 0 1.5 RCIM_Cubic>,<at 0.2 30>,<at 1 0.5>]}
         GreenCurve {keys [<at 0 1.5 RCIM_Cubic>,<at 0.2 30>,<at 1 0.5>]}
         RedCurve {keys [<at 0 1.5 RCIM_Cubic>,<at 0.2 30>,<at 1 0.5>]}
        }}))
     Float=$System.TOD
    )
    ScaleA=true
    ScaleRGB=true
   )
  ]
  rendererClasses [&NiagaraSpriteRendererProperties]
  renderers [{rendererClass=&NiagaraSpriteRendererProperties}]
 }
 Impact_Flare {
  emitterUpdateScript [
   &EmitterState()
   &SpawnBurst_Instantaneous(Spawn Count=1, Spawn Probability=HlslExpression() {hlslExpression="(User.ColorSwitcher==0.0) ? 1.0 : 0.0"}, Use Spawn Probability=true)
  ]
  particleSpawnScript [
   /Niagara/Modules/Spawn/Initialization/V2/InitializeParticle(
    Color=&LerpLinearColors(StartColor=<rgba 1.9089 0.1771 2 1>, EndColor=<rgba 0 0.0182 0.1562 1>, LerpFactor=$User.ColorSwitcher)
    Lifetime=0.25
    Mesh Scale Max=<1.05 1.05 1.05>
    Mesh Scale Min=<0.9375 0.9375 0.9375>
    Mesh Scale Mode=::NewEnumerator2
    Sprite Size=<200 175>
    Sprite Size Mode=::NewEnumerator4
    UsePositionOffset=false
    Write Material Random=true
   ) {module=InitializeParticle001}
   &CameraOffset(Camera Offset Amount=50)
   &InitialMeshOrientation(
    Rotation=&RandomRangeVector(Minimum=<-1 -1 -1>, Maximum=<1 1 1>)
    Use Orientation Vector=true
    Use Rotation Vector=true
   ) {module=InitialMeshOrientation001}
  ]
  particleUpdateScript [
   &ParticleState()
   &ScaleColor(
    Scale Alpha=&FloatFromCurve(FloatCurve=DataInterface() {propertyValues {Curve {keys [{arriveTangent=-2.6501, interpMode=RCIM_Cubic, leaveTangent=-2.6501, tangentMode=RCTM_User, time=0, value=1}, <at 1 0>]}}})
    ScaleA=true
    ScaleRGB=false
   ) {enabled=false}
   &ScaleMeshSize(Scale Factor=&VectorFromFloat(Value=&FloatFromCurve(
      FloatCurve=DataInterface() {propertyValues {Curve {keys [{arriveTangent=1.5592, interpMode=RCIM_Cubic, leaveTangent=1.5592, tangentMode=RCTM_User, time=0, value=0.35}, <at 0.6 1.25>]}}}
      Scale Curve=&FN_FrontEndFloatModifier(Gameplay=1, Front End=0.5)
     )))
   &DynamicMaterialParameters(
    Index 0 Param 1=&FloatFromCurve(FloatCurve=DataInterface() {propertyValues {Curve {keys [<at 0 1 RCIM_Cubic>,<at 0.8 0 RCIM_Cubic>]}}})
    Index 0 Param 2=2
    Index 0 Param 3=1
    Index 0 Param 4=1
    Index 1 Param 1=&FloatFromCurve(FloatCurve=DataInterface() {propertyValues {Curve {keys [
         {arriveTangent=-2.3163, interpMode=RCIM_Cubic, time=0, value=2}
         {arriveTangent=-1.9019, interpMode=RCIM_Cubic, leaveTangent=-1.9019, tangentMode=RCTM_User, time=1, value=0.4}
        ]}}})
    Index 1 Param 2=0.4
    Index 1 Param 3=0.6
    Index 1 Param 4=1
    Index 2 Param 1=&FloatFromCurve(FloatCurve=DataInterface() {propertyValues {Curve {keys [<at 0.46 0>,<at 1 1>]}}})
    Index 2 Param 2=45
    Module.Param0WriteEnabled=true
    Module.Param0WriteEnabled1=true
    Module.Param0WriteEnabled2=true
    Module.Param1WriteEnabled=true
    Module.Param1WriteEnabled1=true
    Module.Param1WriteEnabled2=true
    Module.Param2WriteEnabled=true
    Module.Param2WriteEnabled1=true
    Module.Param2WriteEnabled2=true
    Module.Param3WriteEnabled=true
    Module.Param3WriteEnabled1=true
    Module.Param3WriteEnabled2=true
    Write Parameter Index 1=true
    Write Parameter Index 2=true
   ) {module=DynamicMaterialParameters001}
   &FN_Mesh_Attributes(
    Color=&LerpLinearColors(StartColor=<rgba 0.0105 0.4063 0.5052 1>, EndColor=<rgba 0 0 0 1>, LerpFactor=$User.ColorSwitcher)
    Create Color Attribute=true
    Create Scale Attribute=true
    Scale=&Multiply_VectorByFloat(Vector=$Particles.Scale, Float=1.3)
    Write to Module Namespace=true
   )
  ]
  rendererClasses [&NiagaraMeshRendererProperties]
  renderers [
   {rendererClass=&NiagaraMeshRendererProperties, rendererIndex=0}
   {rendererClass=&NiagaraMeshRendererProperties, rendererIndex=1}
  ]
 }
 Impact_Flare_Variant_01 {
  emitterUpdateScript [
   &EmitterState()
   &SpawnBurst_Instantaneous(Spawn Count=1, Spawn Probability=HlslExpression() {hlslExpression="(User.ColorSwitcher==1.0) ? 1.0 : 0.0"}, Use Spawn Probability=true)
  ]
  particleSpawnScript [
   &InitializeParticle(
    Color=&LerpLinearColors(StartColor=<rgba 1.9089 0.1771 2 1>, EndColor=<rgba 0.3022 1.598 2 0.5>, LerpFactor=$User.ColorSwitcher)
    Lifetime=0.2
    Mesh Scale Max=<1.05 1.05 1.05>
    Mesh Scale Min=<0.9375 0.9375 0.9375>
    Mesh Scale Mode=::NewEnumerator2
    Sprite Size=<200 175>
    Sprite Size Mode=::NewEnumerator4
    UsePositionOffset=false
    Write Material Random=true
   ) {module=InitializeParticle001}
   &CameraOffset(Camera Offset Amount=50)
   &InitialMeshOrientation(
    Rotation=&RandomRangeVector(Minimum=<-1 -1 -1>, Maximum=<1 1 1>)
    Use Orientation Vector=true
    Use Rotation Vector=true
   ) {module=InitialMeshOrientation001}
  ]
  particleUpdateScript [
   &ParticleState()
   &ScaleColor(
    Scale Alpha=&FloatFromCurve(FloatCurve=DataInterface() {propertyValues {Curve {keys [{arriveTangent=-2.6501, interpMode=RCIM_Cubic, leaveTangent=-2.6501, tangentMode=RCTM_User, time=0, value=1}, <at 1 0>]}}})
    ScaleA=true
    ScaleRGB=false
   ) {enabled=false}
   &ScaleMeshSize(Scale Factor=&VectorFromFloat(Value=&FloatFromCurve(
      FloatCurve=DataInterface() {propertyValues {Curve {keys [{arriveTangent=1.5592, interpMode=RCIM_Cubic, leaveTangent=1.5592, tangentMode=RCTM_User, time=0, value=0.35}, <at 0.6 1.25>]}}}
      Scale Curve=&FN_FrontEndFloatModifier(Gameplay=1, Front End=0.5)
     )))
   &DynamicMaterialParameters(
    Index 0 Param 1=&FloatFromCurve(FloatCurve=DataInterface() {propertyValues {Curve {keys [<at 0 1 RCIM_Cubic>,<at 0.8 0 RCIM_Cubic>]}}})
    Index 0 Param 2=2
    Index 0 Param 3=1
    Index 0 Param 4=1
    Index 1 Param 1=&FloatFromCurve(FloatCurve=DataInterface() {propertyValues {Curve {keys [
         {arriveTangent=-2.3163, interpMode=RCIM_Cubic, time=0, value=2}
         {arriveTangent=-1.9019, interpMode=RCIM_Cubic, leaveTangent=-1.9019, tangentMode=RCTM_User, time=1, value=0.4}
        ]}}})
    Index 1 Param 2=0.4
    Index 1 Param 3=0.6
    Index 1 Param 4=1
    Index 2 Param 1=&FloatFromCurve(FloatCurve=DataInterface() {propertyValues {Curve {keys [<at 0.46 0>,<at 1 1>]}}})
    Index 2 Param 2=45
    Module.Param0WriteEnabled=true
    Module.Param0WriteEnabled1=true
    Module.Param0WriteEnabled2=true
    Module.Param1WriteEnabled=true
    Module.Param1WriteEnabled1=true
    Module.Param1WriteEnabled2=true
    Module.Param2WriteEnabled=true
    Module.Param2WriteEnabled1=true
    Module.Param2WriteEnabled2=true
    Module.Param3WriteEnabled=true
    Module.Param3WriteEnabled1=true
    Module.Param3WriteEnabled2=true
    Write Parameter Index 1=true
    Write Parameter Index 2=true
   ) {module=DynamicMaterialParameters001}
   &FN_Mesh_Attributes(
    Color=&LerpLinearColors(StartColor=<rgba 0.0105 0.4063 0.5052 1>, EndColor=<rgba 0 0 0 1>, LerpFactor=$User.ColorSwitcher)
    Create Color Attribute=true
    Create Scale Attribute=true
    Scale=&Multiply_VectorByFloat(Vector=$Particles.Scale, Float=1.3)
    Write to Module Namespace=true
   )
  ]
  rendererClasses [&NiagaraMeshRendererProperties]
  renderers [
   {rendererClass=&NiagaraMeshRendererProperties, rendererIndex=0}
   {rendererClass=&NiagaraMeshRendererProperties, rendererIndex=1}
  ]
 }
 Impact_Outlines {
  emitterUpdateScript [
   &EmitterState(Life Cycle Mode=::NewEnumerator1, Loop Behavior=::NewEnumerator1, Loop Duration=1, UseLoopDelay=false)
   &SpawnBurst_Instantaneous(
    Spawn Count=3
    Spawn Probability=HlslExpression() {hlslExpression="(User.ColorSwitcher==0.0) ? 1.0 : 0.0"}
    Use Spawn Probability=false
    UseLoopCountLimit=false
   ) {module=SpawnBurst_Instantaneous001}
   /SetVariables(Emitter.ColorSeed:NiagaraInt32=&RandomRangeInt(Minimum=0, Maximum=3)) {isSetParameters=true}
  ]
  particleSpawnScript [
   &FN_SetSpawnGroup(Normalize Spawn Group=true, Particles Per Group=1)
   /Niagara/Modules/Spawn/Initialization/InitializeParticle(
    Color=<rgba 1 1 1 10>
    Lifetime=0.26
    Module.Write Color=true
    Module.Write Lifetime=true
    Module.Write Mass=true
    Module.Write Material Random=true
    Module.Write Position=false
    Module.Write Scale=false
    Module.Write SpriteRotation=true
    Module.Write SpriteSize=true
    Position=&SimulationPosition()
    Sprite Size=&Vector2DFromFloat(Value=&Multiply_Float(A=250, B=&RandomRangeFloat(Minimum=0.7, Maximum=1)))
   ) {enabled=false}
   /Niagara/Modules/Spawn/Initialization/V2/InitializeParticle(
    Color=&LerpLinearColors(
     StartColor=&MultiplyLinearColor(
      Linear Color=&SelectLinearColorFromArray(
       Array Sampling Mode=::NewEnumerator1
       Direct Array Index=&Add_Integer(A=$Particles.SpawnGroup, B=$Emitter.ColorSeed)
       Direct Array Mode=::NewEnumerator1
       Color Selection Array=$User.MeshColors
       Random Seed=?
      )
      Scale Factor=<1 1 1 1>
     )
     EndColor=&MultiplyLinearColor(
      Linear Color=&SelectLinearColorFromArray(
       Array Sampling Mode=::NewEnumerator1
       Direct Array Index=&Add_Integer(A=$Particles.SpawnGroup, B=$Emitter.ColorSeed)
       Direct Array Mode=::NewEnumerator1
       Color Selection Array=$User.ImpactVariantColors
       Random Seed=?
      )
      Scale Factor=<1 1 1 1>
     )
     LerpFactor=$User.ColorSwitcher
    )
    Lifetime Max=0.35
    Lifetime Min=0.22
    Lifetime Mode=::NewEnumerator1
    Sprite Rotation Mode=::NewEnumerator2
    Sprite Size Mode=::NewEnumerator3
    Uniform Sprite Size=&SelectFloatFromArray(Direct Array Index=$Particles.SpawnGroup, Float Selection Array=DataInterface() {propertyValues {FloatData [80,50,100,120]}})
    UsePositionOffset=false
    Write Material Random=true
   ) {module=InitializeParticle001}
   /SetVariables(Particles.SpriteSize=&Multiply_Vector2D_ByFloat(Float=&FN_FrontEndFloatModifier(Gameplay=1, Front End=0.6), Vector2D=$Particles.SpriteSize)) {isSetParameters=true}
  ]
  particleUpdateScript [
   &ParticleState()
   &SolveForcesAndVelocity(Clamp Velocity=false, Limit Acceleration=false)
   &ScaleSpriteSize(Scale Factor=&Vector2DFromFloat(Value=&FloatFromCurve(
      FloatCurve=DataInterface() {propertyValues {Curve {keys [
          {arriveTangent=3.4601, interpMode=RCIM_Cubic, leaveTangent=3.4601, tangentMode=RCTM_User, time=0, value=0}
          {arriveTangent=0.2516, interpMode=RCIM_Cubic, leaveTangent=0.2516, tangentMode=RCTM_User, time=1, value=1.4}
         ]}}}
      Scale Curve=0.9
     )))
   &ScaleColor(
    Scale Alpha=&FloatFromCurve(FloatCurve=DataInterface() {propertyValues {Curve {keys [
         {arriveTangent=-3.5345, interpMode=RCIM_Cubic, leaveTangent=-3.5345, tangentMode=RCTM_User, time=0.5, value=1}
         {arriveTangent=-0.1971, interpMode=RCIM_Cubic, leaveTangent=-0.1971, tangentMode=RCTM_User, time=1, value=0}
        ]}}})
    Scale RGB=&VectorFromFloat(Value=$System.TOD)
    ScaleA=true
    ScaleRGB=true
   )
   &CameraOffset(Camera Offset Amount=10)
   &DynamicMaterialParameters(
    Index 0 Param 1=&FloatFromCurve(FloatCurve=DataInterface() {propertyValues {Curve {keys [<at 0.2 -1 RCIM_Cubic>, {arriveTangent=1.7682, interpMode=RCIM_Cubic, leaveTangent=1.7682, tangentMode=RCTM_User, time=1, value=0.4}]}}})
    Module.Param0WriteEnabled=true
    Module.Param1WriteEnabled=true
    Module.Param2WriteEnabled=true
    Module.Param3WriteEnabled=true
   )
  ]
  rendererClasses [&NiagaraSpriteRendererProperties]
  renderers [{rendererClass=&NiagaraSpriteRendererProperties}]
 }
 Plus {
  emitterUpdateScript [
   &EmitterState()
   &SpawnBurst_Instantaneous(
    Spawn Count=&RandomRangeInt(Minimum=1, Maximum=2)
    Spawn Group=1
    Use Spawn Probability=false
    UseLoopCountLimit=false
   )
   &NM_ToonShaderParameters(Lighting_Hardness=1, RimLight_Color:LinearColor=<rgba 1 1 1 1>, RimLight_Width=0)
   &NM_OutlineParameters()
  ]
  particleSpawnScript [
   &PartitionParticles()
   &FN_SetSpawnGroup(Number of Spawn Groups=6)
   &InitializeParticle(
    Color=&LerpLinearColors(
     StartColor=&SelectLinearColorFromArray(Array Sampling Mode=::NewEnumerator0, Direct Array Index=?, Color Selection Array=$User.MeshColors, Random Seed=0)
     EndColor=&SelectLinearColorFromArray(Array Sampling Mode=::NewEnumerator0, Direct Array Index=?, Color Selection Array=$User.MeshColors_Variant_01, Random Seed=0)
     LerpFactor=$User.ColorSwitcher
    )
    Lifetime Max=0.5
    Lifetime Min=0.25
    Lifetime Mode=::NewEnumerator1
    Mesh Scale Mode=::NewEnumerator1
    Mesh Uniform Scale Max=3.5
    Mesh Uniform Scale Min=2
    Position Offset Coordinate Space=::Local
    Sprite UV Mode=::NewEnumerator3
    UsePositionOffset=true
    Write Material Random=true
   ) {module=InitializeParticle001}
   &MeshRotationForce(Rotation=&VectorFromFloat(Value=1))
   &InitialMeshOrientation(Use Rotation Vector=false) {module=InitialMeshOrientation001}
   &AddVelocity(
    Velocity=&RandomRangeVector(Minimum=<200 -1300 -1300>, Maximum=<600 1300 1300>)
    Velocity Speed Scale=&RandomRangeFloat(Minimum=0.3, Maximum=0.8)
   ) {module=AddVelocity001}
   &ApplyInitialForces()
   &CameraOffset(Camera Offset Amount=25)
   /SetVariables(Particles.Color=&MultiplyLinearColor(
     Linear Color=$Particles.Color
     Scale Factor=&MakeFloat4FromVectorAndFloat("Vector (RGB)"=&VectorFromFloat(Value=1), "Float (Alpha)"=1)
    )) {isSetParameters=true}
   /SetVariables(Particles.Scale=&Multiply_VectorByFloat(Vector=$Particles.Scale, Float=&FN_FrontEndFloatModifier(Gameplay=1, Front End=0.3))) {isSetParameters=true}
  ]
  particleUpdateScript [
   &ParticleState()
   &Drag(Drag=&RandomRangeFloat(Minimum=8, Maximum=15), Use Linear Drag=true, Use Rotational Drag=false) {module=Drag001}
   &SpriteRotationRate(Rotation Rate=&RandomRangeFloat(Minimum=50, Maximum=170))
   &GravityForce(Gravity=&RandomRangeVector(Minimum=<0 0 150>, Maximum=<0 0 -150>))
   &CurlNoiseForce(Pan Noise=false)
   &ScaleSpriteSize(Uniform Curve Sprite Scale=DataInterface() {propertyValues {Curve {keys [
        {interpMode=RCIM_Cubic, tangentMode=RCTM_User, time=0, value=0}
        <at 0.2 1>
        {interpMode=RCIM_Cubic, tangentMode=RCTM_User, time=0.7, value=1}
        {arriveTangent=-0.9819, interpMode=RCIM_Cubic, leaveTangent=-0.9819, tangentMode=RCTM_User, time=0.85, value=1.3}
        {arriveTangent=-9.5654, interpMode=RCIM_Cubic, leaveTangent=-9.5654, tangentMode=RCTM_User, time=1, value=0}
       ]}}})
   &SolveForcesAndVelocity(Clamp Velocity=false, Limit Acceleration=false)
   /SetVariables(
    Particles.Position=&AddVectorToPosition(
     Position=$Particles.Position
     Vector=&Multiply_Vector(
      A=&VectorFromFloat(Value=&FN_FrontEndFloatModifier(Gameplay=0.8, Front End=0.1))
      B=&SelectVectorFromArray(
       Direct Array Index=&ModuloInt(
        A=&MakeIntFromFloat(Float=&Add_Float(
          A=&Multiply_Float(A=$Engine.Time, B=6)
          B=&MakeFloatFromInt(Int=$Particles.SpawnGroup)
         ))
        B=6
       )
       Vector Selection Array=DataInterface() {propertyValues {FloatData [<1 0 0>, <0 -0.5 -0.5>, <-1 0.5 0.5>, <0.5 2 1>, <0.1 -0.4 0.2>, <0.35 0.7 -0.4>]}}
      )
     )
    )
    Particles.Scale=&Multiply_Vector(
     A=$Particles.Scale
     B=&SelectVectorFromArray(
      Direct Array Index=&ModuloInt(
       A=&MakeIntFromFloat(Float=&Add_Float(
         A=&Multiply_Float(A=$Engine.Time, B=6)
         B=&MakeFloatFromInt(Int=$Particles.SpawnGroup)
        ))
       B=3
      )
      Vector Selection Array=DataInterface() {propertyValues {FloatData [<1 1 1>, <1.05 1 1.05>, <0.95 1.02 0.99>]}}
     )
    )
   ) {isSetParameters=true}
  ]
  rendererClasses [&NiagaraMeshRendererProperties]
  renderers [
   {rendererClass=&NiagaraMeshRendererProperties, rendererIndex=0}
   {rendererClass=&NiagaraMeshRendererProperties, rendererIndex=1}
  ]
 }
 Smoke002 {
  emitterUpdateScript [
   &EmitterState()
   &SpawnBurst_Instantaneous(Spawn Count=1, Spawn Time=0.03, Use Spawn Probability=false)
  ]
  particleSpawnScript [
   /SetVariables(Particles.Random=&RandomRangeFloat(Minimum=0, Maximum=1)) {isSetParameters=true, module=SetVariables_#2}
   /Niagara/Modules/Spawn/Initialization/InitializeParticle(
    Color=&LerpLinearColors(StartColor=<rgba 0.074 0.0995 0.1406 0.2>, EndColor=<rgba 0.1125 0.1233 0.1406 0.2>, LerpFactor=$User.ColorSwitcher)
    Lifetime=&Lerp_Float(A=0.35, B=0.3, Alpha=$Particles.Random)
    Module.Write Color=true
    Module.Write Lifetime=true
    Module.Write Mass=true
    Module.Write Material Random=true
    Module.Write Position=false
    Module.Write Scale=false
    Module.Write SpriteRotation=true
    Module.Write SpriteSize=true
    Position=&SimulationPosition()
    Sprite Rotation=&Lerp_Float(A=-45, B=45, Alpha=$Particles.Random)
    Sprite Size=&Multiply_Vector2D_ByFloat(
     Float=&FN_FrontEndFloatModifier(Gameplay=1, Front End=0.5)
     Vector2D=&Vector2DFromFloat(Value=&Lerp_Float(A=120, B=140, Alpha=$Particles.Random))
    )
   )
   &CameraOffset(Camera Offset Amount=30)
   &DynamicMaterialParameters(
    Index 0 Param 2=&UniformAOrBFloat(A=0)
    Module.Param0WriteEnabled=true
    Module.Param1WriteEnabled=true
    Module.Param2WriteEnabled=true
    Module.Param3WriteEnabled=true
   ) {module=DynamicMaterialParameters001}
  ]
  particleUpdateScript [
   &ParticleState()
   &SolveForcesAndVelocity(Clamp Velocity=false, Limit Acceleration=false)
   &DynamicMaterialParameters(
    Index 0 Param 1=&FloatFromCurve(FloatCurve=DataInterface() {propertyValues {Curve {keys [<at 0 0>,<at 1 1>]}}})
    Index 0 Param 2=1
    Module.Param0WriteEnabled=true
    Module.Param1WriteEnabled=true
    Module.Param2WriteEnabled=true
    Module.Param3WriteEnabled=true
   )
   &ScaleColor(
    Scale Alpha=&FloatFromCurve(FloatCurve=DataInterface() {propertyValues {Curve {keys [<at 0 1>,<at 0.8091 0.9932>,<at 1 0>]}}}, Scale Curve=4)
    ScaleA=true
    ScaleRGB=true
   )
  ]
  rendererClasses [&NiagaraSpriteRendererProperties]
  renderers [{rendererClass=&NiagaraSpriteRendererProperties}]
 }
 Sparks003 {
  emitterUpdateScript [
   &EmitterState()
   &SpawnBurst_Instantaneous(Spawn Count=&RandomRangeInt(Minimum=2, Maximum=5), Use Spawn Probability=false)
   &SpawnBurst_Instantaneous(Spawn Count=&RandomRangeInt(Minimum=8, Maximum=10), Spawn Time=0.05, Use Spawn Probability=false) {module=SpawnBurst_Instantaneous001}
  ]
  particleSpawnScript [
   /SetVariables(Particles.Random=&RandomRangeFloat(Minimum=0, Maximum=1)) {isSetParameters=true, module=SetVariables_#2}
   &InitializeParticle(
    Color=&LerpLinearColors(StartColor=<rgba 0.3073 0.0768 0.2189 1>, EndColor=<rgba 0.0768 0.1037 0.3073 1>, LerpFactor=$User.ColorSwitcher)
    Lifetime=&Lerp_Float(A=0.1, B=0.3, Alpha=$Particles.Random)
    Mesh Scale=&Multiply_VectorByFloat(
     Vector=&Multiply_Vector(A=&Lerp_Vector(A=<0.2 0.2 0.2>, B=<0.6 0.6 0.6>, Alpha=$Particles.Random), B=<0.8 0.8 0.4>)
     Float=&FN_FrontEndFloatModifier(Gameplay=1, Front End=0.5)
    )
    Module.Write Color=true
    Module.Write Lifetime=true
    Module.Write Mass=true
    Module.Write Material Random=true
    Module.Write Position=false
    Module.Write Scale=true
    Module.Write SpriteRotation=false
    Module.Write SpriteSize=true
    Position=&SimulationPosition()
    Sprite Size=&RandomRangeVector2D(Minimum=<5 40>, Maximum=<10 65>)
   )
   &CameraOffset(Camera Offset Amount=30)
   /Niagara/Modules/Spawn/Location/V2/SphereLocation(
    MaskSpawnGroup=false
    Override Seed=false
    Sphere Radius=1
    Surface Only=false
    UseNonUniformScale=false
    UseOffset=false
    UseSphereOrientation=false
   )
   &AddVelocity(
    Scale Added Velocity=&Multiply_VectorByFloat(Vector=<1 1 1>, Float=&Lerp_Float(A=2, B=3, Alpha=$Particles.Random))
    Velocity=$Output.SphereLocation.SphereNormal
   )
   &DynamicMaterialParameters(
    Index 0 Param 2=&FloatFromCurve(
     FloatCurve=DataInterface() {propertyValues {Curve {keys [<at 0 0 RCIM_Cubic>, {arriveTangent=1, interpMode=RCIM_Cubic, leaveTangent=1, time=0.7035, value=0.3698}, <at 1 1>]}}}
     CurveIndex=&RandomRangeFloat(Minimum=0, Maximum=1)
    )
    Module.Param0WriteEnabled=true
    Module.Param1WriteEnabled=true
    Module.Param2WriteEnabled=true
    Module.Param3WriteEnabled=true
   ) {module=DynamicMaterialParameters001}
   &InitialMeshOrientation(
    Orientation Vector=<1 1 1>
    Rotation=&RandomRangeVector(Minimum=<-1 -1 -1>, Maximum=<1 1 1>)
    Use Orientation Vector=true
    Use Rotation Vector=true
   )
  ]
  particleUpdateScript [
   &ParticleState()
   &SolveForcesAndVelocity(Clamp Velocity=false, Limit Acceleration=false)
   &DynamicMaterialParameters(
    Index 0 Param 1=&FloatFromCurve(FloatCurve=DataInterface() {propertyValues {Curve {keys [<at 0 0>,<at 1 1>]}}})
    Module.Param0WriteEnabled=true
    Module.Param1WriteEnabled=false
    Module.Param2WriteEnabled=true
    Module.Param3WriteEnabled=true
   )
  ]
  rendererClasses [&NiagaraSpriteRendererProperties, &NiagaraMeshRendererProperties]
  renderers [
   {rendererClass=&NiagaraSpriteRendererProperties}
   {rendererClass=&NiagaraMeshRendererProperties, rendererIndex=1}
  ]
 }
 Stars {
  emitterUpdateScript [
   &EmitterState()
   &SpawnBurst_Instantaneous(
    Spawn Count=&RandomRangeInt(Minimum=1, Maximum=2)
    Spawn Group=1
    Use Spawn Probability=false
    UseLoopCountLimit=false
   )
   &NM_ToonShaderParameters(Lighting_Hardness=0, RimLight_Color:LinearColor=<rgba 1 1 1 1>, RimLight_Width=0)
   &NM_OutlineParameters(OutlineWidth=0.7)
  ]
  particleSpawnScript [
   &PartitionParticles()
   &FN_SetSpawnGroup(Number of Spawn Groups=6)
   /Niagara/Modules/Spawn/Initialization/V2/InitializeParticle(
    Color=&LerpLinearColors(
     StartColor=&SelectLinearColorFromArray(Array Sampling Mode=::NewEnumerator0, Direct Array Index=?, Color Selection Array=$User.MeshColors, Random Seed=0)
     EndColor=&SelectLinearColorFromArray(Array Sampling Mode=::NewEnumerator0, Direct Array Index=?, Color Selection Array=$User.MeshColors_Variant_01, Random Seed=0)
     LerpFactor=$User.ColorSwitcher
    )
    Lifetime Max=0.5
    Lifetime Min=0.25
    Lifetime Mode=::NewEnumerator1
    Mesh Scale Mode=::NewEnumerator1
    Mesh Uniform Scale Max=3.5
    Mesh Uniform Scale Min=3
    Position Offset Coordinate Space=::Local
    Sprite UV Mode=::NewEnumerator3
    UsePositionOffset=true
    Write Material Random=true
   )
   &MeshRotationForce(Rotation=&VectorFromFloat(Value=1))
   &InitialMeshOrientation(Use Rotation Vector=false) {module=InitialMeshOrientation001}
   &AddVelocity(
    Velocity=&RandomRangeVector(Minimum=<200 -1300 -1300>, Maximum=<600 1300 1300>)
    Velocity Speed Scale=&RandomRangeFloat(Minimum=0.3, Maximum=0.8)
   ) {module=AddVelocity001}
   &ApplyInitialForces()
   &CameraOffset(Camera Offset Amount=25)
   /SetVariables(Particles.Color=&MultiplyLinearColor(
     Linear Color=$Particles.Color
     Scale Factor=&MakeFloat4FromVectorAndFloat("Vector (RGB)"=&VectorFromFloat(Value=1), "Float (Alpha)"=1)
    )) {isSetParameters=true}
   /SetVariables(Particles.Scale=&Multiply_VectorByFloat(Vector=$Particles.Scale, Float=&FN_FrontEndFloatModifier(Gameplay=1, Front End=0.3))) {isSetParameters=true}
  ]
  particleUpdateScript [
   &ParticleState()
   &ScaleSpriteSize(Uniform Curve Sprite Scale=DataInterface() {propertyValues {Curve {keys [
        {interpMode=RCIM_Cubic, tangentMode=RCTM_User, time=0, value=0}
        <at 0.2 1>
        {interpMode=RCIM_Cubic, tangentMode=RCTM_User, time=0.7, value=1}
        {arriveTangent=-0.9819, interpMode=RCIM_Cubic, leaveTangent=-0.9819, tangentMode=RCTM_User, time=0.85, value=1.3}
        {arriveTangent=-9.5654, interpMode=RCIM_Cubic, leaveTangent=-9.5654, tangentMode=RCTM_User, time=1, value=0}
       ]}}})
   &ScaleMeshSize(Scale Factor=&Multiply_VectorByFloat(
     Vector=<1 1 1>
     Float=&FloatFromCurve(FloatCurve=DataInterface() {propertyValues {Curve {keys [
          {interpMode=RCIM_Cubic, tangentMode=RCTM_User, time=0, value=0}
          <at 0.2 1>
          {interpMode=RCIM_Cubic, tangentMode=RCTM_User, time=0.8, value=1}
          {tangentMode=RCTM_User, time=1, value=0}
         ]}}})
    )) {module=ScaleMeshSize006}
   &SpriteRotationRate(Rotation Rate=&RandomRangeFloat(Minimum=50, Maximum=170))
   &Drag(Drag=&RandomRangeFloat(Minimum=8, Maximum=15), Use Linear Drag=true, Use Rotational Drag=false) {module=Drag001}
   &GravityForce(Gravity=&RandomRangeVector(Minimum=<0 0 150>, Maximum=<0 0 -150>))
   &CurlNoiseForce(Pan Noise=false)
   &SolveForcesAndVelocity(Clamp Velocity=false, Limit Acceleration=false)
   /SetVariables(
    Particles.Position=&AddVectorToPosition(
     Position=$Particles.Position
     Vector=&Multiply_Vector(
      A=&VectorFromFloat(Value=&FN_FrontEndFloatModifier(Gameplay=0.8, Front End=0.1))
      B=&SelectVectorFromArray(
       Direct Array Index=&ModuloInt(
        A=&MakeIntFromFloat(Float=&Add_Float(
          A=&Multiply_Float(A=$Engine.Time, B=6)
          B=&MakeFloatFromInt(Int=$Particles.SpawnGroup)
         ))
        B=6
       )
       Vector Selection Array=DataInterface() {propertyValues {FloatData [<1 0 0>, <0 -0.5 -0.5>, <-1 0.5 0.5>, <0.5 2 1>, <0.1 -0.4 0.2>, <0.35 0.7 -0.4>]}}
      )
     )
    )
    Particles.Scale=&Multiply_Vector(
     A=$Particles.Scale
     B=&SelectVectorFromArray(
      Direct Array Index=&ModuloInt(
       A=&MakeIntFromFloat(Float=&Add_Float(
         A=&Multiply_Float(A=$Engine.Time, B=6)
         B=&MakeFloatFromInt(Int=$Particles.SpawnGroup)
        ))
       B=3
      )
      Vector Selection Array=DataInterface() {propertyValues {FloatData [<1 1 1>, <1.05 1 1.05>, <0.95 1.02 0.99>]}}
     )
    )
   ) {isSetParameters=true}
  ]
  rendererClasses [&NiagaraMeshRendererProperties]
  renderers [
   {rendererClass=&NiagaraMeshRendererProperties, rendererIndex=0}
   {rendererClass=&NiagaraMeshRendererProperties, rendererIndex=1}
  ]
 }
 Zaps {
  emitterUpdateScript [
   &EmitterState()
   &SpawnBurst_Instantaneous(
    Spawn Count=&RandomRangeInt(Minimum=1, Maximum=2)
    Spawn Group=1
    Use Spawn Probability=false
    UseLoopCountLimit=false
   )
   &NM_ToonShaderParameters(Lighting_Hardness=1, RimLight_Color:LinearColor=<rgba 1 1 1 1>, RimLight_Width=0)
   &NM_OutlineParameters()
  ]
  particleSpawnScript [
   &PartitionParticles()
   &FN_SetSpawnGroup(Number of Spawn Groups=6)
   &InitializeParticle(
    Color=&LerpLinearColors(
     StartColor=&SelectLinearColorFromArray(Array Sampling Mode=::NewEnumerator0, Direct Array Index=?, Color Selection Array=$User.MeshColors, Random Seed=0)
     EndColor=&SelectLinearColorFromArray(Array Sampling Mode=::NewEnumerator0, Direct Array Index=?, Color Selection Array=$User.MeshColors_Variant_01, Random Seed=0)
     LerpFactor=$User.ColorSwitcher
    )
    Lifetime Max=0.5
    Lifetime Min=0.25
    Lifetime Mode=::NewEnumerator1
    Mesh Scale Mode=::NewEnumerator1
    Mesh Uniform Scale Max=4.5
    Mesh Uniform Scale Min=3
    Position Offset Coordinate Space=::Local
    Sprite UV Mode=::NewEnumerator3
    UsePositionOffset=true
    Write Material Random=true
   )
   &MeshRotationForce(Rotation=&VectorFromFloat(Value=1))
   &InitialMeshOrientation(Use Rotation Vector=false) {module=InitialMeshOrientation001}
   &AddVelocity(
    Velocity=&RandomRangeVector(Minimum=<200 -1300 -1300>, Maximum=<600 1300 1300>)
    Velocity Speed Scale=&RandomRangeFloat(Minimum=0.3, Maximum=0.8)
   ) {module=AddVelocity001}
   &ApplyInitialForces()
   &CameraOffset(Camera Offset Amount=25)
   /SetVariables(Particles.CustomMeshIndex:NiagaraInt32=&RandomRangeInt(Minimum=0, Maximum=1)) {isSetParameters=true}
   /SetVariables(Particles.Color=&MultiplyLinearColor(
     Linear Color=$Particles.Color
     Scale Factor=&MakeFloat4FromVectorAndFloat("Vector (RGB)"=&VectorFromFloat(Value=1), "Float (Alpha)"=1)
    )) {isSetParameters=true}
   /SetVariables(Particles.Scale=&Multiply_VectorByFloat(Vector=$Particles.Scale, Float=&FN_FrontEndFloatModifier(Gameplay=1, Front End=0.3))) {isSetParameters=true}
  ]
  particleUpdateScript [
   &ParticleState()
   &SpriteRotationRate(Rotation Rate=&RandomRangeFloat(Minimum=50, Maximum=170))
   &ScaleMeshSize(Scale Factor=&Multiply_VectorByFloat(
     Vector=<1 1 1>
     Float=&FloatFromCurve(FloatCurve=DataInterface() {propertyValues {Curve {keys [
          {interpMode=RCIM_Cubic, tangentMode=RCTM_User, time=0, value=0}
          <at 0.2 1>
          {interpMode=RCIM_Cubic, tangentMode=RCTM_User, time=0.8, value=1}
          {tangentMode=RCTM_User, time=1, value=0}
         ]}}})
    )) {module=ScaleMeshSize006}
   &ScaleSpriteSize(Uniform Curve Sprite Scale=DataInterface() {propertyValues {Curve {keys [
        {interpMode=RCIM_Cubic, tangentMode=RCTM_User, time=0, value=0}
        <at 0.2 1>
        {interpMode=RCIM_Cubic, tangentMode=RCTM_User, time=0.7, value=1}
        {arriveTangent=-0.9819, interpMode=RCIM_Cubic, leaveTangent=-0.9819, tangentMode=RCTM_User, time=0.85, value=1.3}
        {arriveTangent=-9.5654, interpMode=RCIM_Cubic, leaveTangent=-9.5654, tangentMode=RCTM_User, time=1, value=0}
       ]}}})
   &Drag(Drag=&RandomRangeFloat(Minimum=8, Maximum=15), Use Linear Drag=true, Use Rotational Drag=false) {module=Drag001}
   &CurlNoiseForce(Pan Noise=false)
   &GravityForce(Gravity=&RandomRangeVector(Minimum=<0 0 150>, Maximum=<0 0 -150>))
   &SolveForcesAndVelocity(Clamp Velocity=false, Limit Acceleration=false)
   /SetVariables(
    Particles.Position=&AddVectorToPosition(
     Position=$Particles.Position
     Vector=&Multiply_Vector(
      A=&VectorFromFloat(Value=&FN_FrontEndFloatModifier(Gameplay=0.8, Front End=0.1))
      B=&SelectVectorFromArray(
       Direct Array Index=&ModuloInt(
        A=&MakeIntFromFloat(Float=&Add_Float(
          A=&Multiply_Float(A=$Engine.Time, B=6)
          B=&MakeFloatFromInt(Int=$Particles.SpawnGroup)
         ))
        B=6
       )
       Vector Selection Array=DataInterface() {propertyValues {FloatData [<1 0 0>, <0 -0.5 -0.5>, <-1 0.5 0.5>, <0.5 2 1>, <0.1 -0.4 0.2>, <0.35 0.7 -0.4>]}}
      )
     )
    )
    Particles.Scale=&Multiply_Vector(
     A=$Particles.Scale
     B=&SelectVectorFromArray(
      Direct Array Index=&ModuloInt(
       A=&MakeIntFromFloat(Float=&Add_Float(
         A=&Multiply_Float(A=$Engine.Time, B=6)
         B=&MakeFloatFromInt(Int=$Particles.SpawnGroup)
        ))
       B=3
      )
      Vector Selection Array=DataInterface() {propertyValues {FloatData [<1 1 1>, <1.05 1 1.05>, <0.95 1.02 0.99>]}}
     )
    )
   ) {isSetParameters=true}
  ]
  rendererClasses [&NiagaraMeshRendererProperties]
  renderers [
   {rendererClass=&NiagaraMeshRendererProperties, rendererIndex=0}
   {rendererClass=&NiagaraMeshRendererProperties, rendererIndex=1}
  ]
 }
}
summary {
 systemName=NS_Pickaxe_Dazzle_Impact
 emitters [
  {emitterName=Impact_Flare, rendererClasses [&NiagaraMeshRendererProperties]}
  {emitterName=Flare, rendererClasses [&NiagaraSpriteRendererProperties]}
  {emitterName=Smoke002, rendererClasses [&NiagaraSpriteRendererProperties]}
  {emitterName=Sparks003, rendererClasses [&NiagaraSpriteRendererProperties, &NiagaraMeshRendererProperties]}
  {emitterName=GlowDrops, rendererClasses [&NiagaraSpriteRendererProperties]}
  {emitterName=Impact_Outlines, rendererClasses [&NiagaraSpriteRendererProperties]}
  {emitterName=Zaps, rendererClasses [&NiagaraMeshRendererProperties]}
  {emitterName=Stars, rendererClasses [&NiagaraMeshRendererProperties]}
  {emitterName=Plus, rendererClasses [&NiagaraMeshRendererProperties]}
  {emitterName=Bunny, rendererClasses [&NiagaraMeshRendererProperties]}
  {emitterName=Emote, rendererClasses [&NiagaraMeshRendererProperties]}
  {emitterName=Impact_Flare_Variant_01, rendererClasses [&NiagaraMeshRendererProperties]}
 ]
 userVariables [
  User.bIsFrontEnd {type=NiagaraBool}
  User.bIsFrontEndPreview {type=NiagaraBool}
  User.ColorSwitcher {type=NiagaraFloat, defaultValue=0, description=""}
  User.EmissiveIntensity {type=NiagaraFloat, defaultValue=0, description=""}
  User.ImpactVariantColors {
   type=NiagaraDataInterfaceArrayColor
   description=""
   defaultValue {
    struct=NiagaraExt_VariableValue_DataInterface
    value {
     dataInterface=&self:NiagaraDataInterfaceArrayColor_2
     dataInterfaceClass=&NiagaraDataInterfaceArrayColor
    }
   }
  }
  User.IsSwing {type=NiagaraBool}
  User.KillStage {type=NiagaraInt32}
  User.MeshColors {
   type=NiagaraDataInterfaceArrayColor
   description=""
   defaultValue {
    struct=NiagaraExt_VariableValue_DataInterface
    value {
     dataInterface=&self:NiagaraDataInterfaceArrayColor_0
     dataInterfaceClass=&NiagaraDataInterfaceArrayColor
    }
   }
  }
  User.MeshColors_Variant_01 {
   type=NiagaraDataInterfaceArrayColor
   description=""
   defaultValue {
    struct=NiagaraExt_VariableValue_DataInterface
    value {
     dataInterface=&self:NiagaraDataInterfaceArrayColor_1
     dataInterfaceClass=&NiagaraDataInterfaceArrayColor
    }
   }
  }
  User.Stage4JitterSpeedMultiplier {type=NiagaraFloat, defaultValue=0, description=""}
  User.VariantSwitch {type=NiagaraFloat, defaultValue=0, description=""}
 ]
}
