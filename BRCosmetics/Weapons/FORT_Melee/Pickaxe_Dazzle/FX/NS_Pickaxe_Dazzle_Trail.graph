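# Niagara trail effect for the Dazzle pickaxe cosmetic: pink and blue anim-trail ribbons, brief sprite shapes around the pickaxe, and toon-shaded, outlined mesh confetti.
# Bunny, emote, plus, star and zap meshes spawn per distance moved, inherit velocity, then drift under drag, gravity and curl noise.
# A time-indexed array lookup, offset by spawn group, gives meshes stepped position and scale jitter.
# User.ColorSwitcher blends colors between two palettes, and System.TOD scales brightness.
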NiagaraSystem /BRCosmetics/Weapons/FORT_Melee/Pickaxe_Dazzle/FX/NS_Pickaxe_Dazzle_Trail
dependencies {
 usedDataInterfaces [
  &NiagaraDataInterfaceCurve
  &NiagaraDataInterfaceArrayFloat3
  &NiagaraDataInterfaceArrayFloat
  &NiagaraDataInterfaceArrayFloat2
 ]
 usedDynamicInputs [
  &SelectFloatFromArray
  &ReturnExecIndex
  &FN_Select_Float_from_Time_Of_Day
 ]
 usedModules [
  &self:SystemSpawnScript.SystemScriptSource.SystemScriptGraph.NiagaraNodeAssignment_0.SetVariables_#1
  &SystemState
  &Timeline
  &self:SystemSpawnScript.SystemScriptSource.SystemScriptGraph.NiagaraNodeAssignment_1.SetVariables_#2
 ]
}
emitters {
 Bunny001 {
  emitterUpdateScript [
   &EmitterState()
   &SpawnPerUnit(Module.Use Max Movement Threshold=true, Module.Use Movement Tolerance=false, Spawn Per Unit=120, Use Spawn Probability=false)
   /Game/Effects/Niagara/MasterMaterials/ToonShader/Template/NM_ToonShaderParameters(Lighting_Hardness=1, RimLight_Color:LinearColor=<rgba 1 1 1 1>, RimLight_Width=0)
   /Game/Effects/Niagara/MasterMaterials/ToonShader/Template/NM_OutlineParameters(OutlineWidth=0.7)
  ]
  particleSpawnScript [
   &PartitionParticles()
   &FN_SetSpawnGroup(Number of Spawn Groups=6)
   &InitializeParticle(
    Color=&LerpLinearColors(
     StartColor=&SelectLinearColorFromArray(Array Sampling Mode=::NewEnumerator0, Direct Array Index=?, Color Selection Array=$User.MeshColors, Random Seed=0)
     EndColor=&SelectLinearColorFromArray(Array Sampling Mode=::NewEnumerator0, Direct Array Index=?, Color Selection Array=$User.MeshColors_Variant_01, Random Seed=0)
     LerpFactor=$User.ColorSwitcher
    )
    Lifetime Max=0.5
    Lifetime Min=0.25
    Lifetime Mode=::NewEnumerator1
    Mesh Scale Mode=::NewEnumerator1
    Mesh Uniform Scale Max=2.5
    Position Offset Coordinate Space=::Local
    Sprite UV Mode=::NewEnumerator3
    UsePositionOffset=true
    Write Material Random=true
   ) {module=InitializeParticle001}
   &MeshRotationForce(Rotation=&VectorFromFloat(Value=2))
   &InitialMeshOrientation(Use Rotation Vector=false) {module=InitialMeshOrientation001}
   &AddVelocity(Velocity=&RandomRangeVector(Minimum=<-1 -1 -1>, Maximum=<1 1 1>), Velocity Speed Scale=10) {module=AddVelocity002}
   &InheritVelocity(
    Inherited Velocity Amount Scale=&Multiply_VectorByFloat(
     Vector=<1 1 1>
     Float=&Multiply_Float(A=$System.InheritVelocityElements, B=&RandomRangeFloat(Minimum=0.5, Maximum=1))
    )
    Inherited Velocity Speed Limit=50
    Speed Limit=true
    UseSpeedThreshold=false
   )
   &ApplyInitialForces()
   &CameraOffset(Camera Offset Amount=25)
   /SetVariables(Particles.Color=&MultiplyLinearColor(
     Linear Color=$Particles.Color
     Scale Factor=&MakeFloat4FromVectorAndFloat("Vector (RGB)"=&VectorFromFloat(Value=$System.TOD), "Float (Alpha)"=1)
    )) {isSetParameters=true}
  ]
  particleUpdateScript [
   &ParticleState()
   &Drag(Drag=&RandomRangeFloat(Minimum=4, Maximum=12), Use Linear Drag=true, Use Rotational Drag=false)
   &GravityForce(Gravity=&RandomRangeVector(Minimum=<0 0 150>, Maximum=<0 0 -150>)) {module=GravityForce001}
   &CurlNoiseForce(Pan Noise=false) {module=CurlNoiseForce001}
   &ScaleMeshSize(Scale Factor=&Multiply_VectorByFloat(
     Vector=<1 1 1>
     Float=&FloatFromCurve(FloatCurve=DataInterface() {propertyValues {Curve {keys [
          {interpMode=RCIM_Cubic, tangentMode=RCTM_User, time=0, value=0}
          <at 0.2 1>
          {interpMode=RCIM_Cubic, tangentMode=RCTM_Break, time=0.7, value=1}
          {arriveTangent=0.1352, interpMode=RCIM_Cubic, leaveTangent=0.1352, tangentMode=RCTM_User, time=0.85, value=1}
          {interpMode=RCIM_Cubic, leaveTangent=-9.5654, time=1, value=0}
         ]}}})
    )) {module=ScaleMeshSize001}
   &SolveForcesAndVelocity(Clamp Velocity=false, Limit Acceleration=false)
   /SetVariables(
    Particles.Position=&AddVectorToPosition(
     Position=$Particles.Position
     Vector=&Multiply_Vector(
      A=&VectorFromFloat(Value=0.6)
      B=&SelectVectorFromArray(
       Direct Array Index=&ModuloInt(
        A=&MakeIntFromFloat(Float=&Add_Float(
          A=&Multiply_Float(A=$Engine.Time, B=6)
          B=&MakeFloatFromInt(Int=$Particles.SpawnGroup)
         ))
        B=6
       )
       Vector Selection Array=DataInterface() {propertyValues {FloatData [<1 0 0>, <0 -0.5 -0.5>, <-1 0.5 0.5>, <0.5 2 1>, <0.1 -0.4 0.2>, <0.35 0.7 -0.4>]}}
      )
     )
    )
    Particles.Scale=&Multiply_Vector(
     A=$Particles.Scale
     B=&SelectVectorFromArray(
      Direct Array Index=&ModuloInt(
       A=&MakeIntFromFloat(Float=&Add_Float(
         A=&Multiply_Float(A=$Engine.Time, B=6)
         B=&MakeFloatFromInt(Int=$Particles.SpawnGroup)
        ))
       B=3
      )
      Vector Selection Array=DataInterface() {propertyValues {FloatData [<1 1 1>, <1.05 1 1.05>, <0.95 1.02 0.99>]}}
     )
    )
   ) {isSetParameters=true}
  ]
  rendererClasses [&NiagaraMeshRendererProperties]
  renderers [
   {rendererClass=&NiagaraMeshRendererProperties, rendererIndex=0}
   {rendererClass=&NiagaraMeshRendererProperties, rendererIndex=1}
  ]
 }
 Emote {
  emitterUpdateScript [
   &EmitterState()
   &SpawnPerUnit(Module.Use Max Movement Threshold=true, Module.Use Movement Tolerance=false, Spawn Per Unit=130, Use Spawn Probability=false)
   &NM_ToonShaderParameters(Lighting_Hardness=1, RimLight_Color:LinearColor=<rgba 1 1 1 1>, RimLight_Width=0)
   &NM_OutlineParameters(OutlineWidth=0.7)
  ]
  particleSpawnScript [
   &PartitionParticles()
   &FN_SetSpawnGroup(Number of Spawn Groups=6)
   &InitializeParticle(
    Color=&LerpLinearColors(
     StartColor=&SelectLinearColorFromArray(Array Sampling Mode=::NewEnumerator0, Direct Array Index=?, Color Selection Array=$User.MeshColors, Random Seed=0)
     EndColor=&SelectLinearColorFromArray(Array Sampling Mode=::NewEnumerator0, Direct Array Index=?, Color Selection Array=$User.MeshColors_Variant_01, Random Seed=0)
     LerpFactor=$User.ColorSwitcher
    )
    Lifetime Max=0.5
    Lifetime Min=0.25
    Lifetime Mode=::NewEnumerator1
    Mesh Scale Mode=::NewEnumerator1
    Mesh Uniform Scale Max=2.5
    Mesh Uniform Scale Min=1.5
    Position Offset Coordinate Space=::Local
    Sprite UV Mode=::NewEnumerator3
    UsePositionOffset=true
    Write Material Random=true
   ) {module=InitializeParticle001}
   &MeshRotationForce(Rotation=&VectorFromFloat(Value=2))
   &InitialMeshOrientation(Use Rotation Vector=false) {module=InitialMeshOrientation001}
   &AddVelocity(Velocity=&RandomRangeVector(Minimum=<-1 -1 -1>, Maximum=<1 1 1>), Velocity Speed Scale=10) {module=AddVelocity002}
   &InheritVelocity(
    Inherited Velocity Amount Scale=&Multiply_VectorByFloat(
     Vector=<1 1 1>
     Float=&Multiply_Float(A=$System.InheritVelocityElements, B=&RandomRangeFloat(Minimum=0.5, Maximum=1))
    )
    Inherited Velocity Speed Limit=50
    Speed Limit=true
    UseSpeedThreshold=false
   )
   &ApplyInitialForces()
   &CameraOffset(Camera Offset Amount=25)
   /SetVariables(Particles.Color=&MultiplyLinearColor(
     Linear Color=$Particles.Color
     Scale Factor=&MakeFloat4FromVectorAndFloat("Vector (RGB)"=&VectorFromFloat(Value=$System.TOD), "Float (Alpha)"=1)
    )) {isSetParameters=true}
  ]
  particleUpdateScript [
   &ParticleState()
   &Drag(Drag=&RandomRangeFloat(Minimum=4, Maximum=12), Use Linear Drag=true, Use Rotational Drag=false)
   &GravityForce(Gravity=&RandomRangeVector(Minimum=<0 0 150>, Maximum=<0 0 -150>)) {module=GravityForce001}
   &CurlNoiseForce(Pan Noise=false) {module=CurlNoiseForce001}
   &ScaleMeshSize(Scale Factor=&Multiply_VectorByFloat(
     Vector=<1 1 1>
     Float=&FloatFromCurve(FloatCurve=DataInterface() {propertyValues {Curve {keys [
          {interpMode=RCIM_Cubic, tangentMode=RCTM_User, time=0, value=0}
          <at 0.2 1>
          {interpMode=RCIM_Cubic, tangentMode=RCTM_Break, time=0.7, value=1}
          {arriveTangent=0.1352, interpMode=RCIM_Cubic, leaveTangent=0.1352, tangentMode=RCTM_User, time=0.85, value=1}
          {interpMode=RCIM_Cubic, leaveTangent=-9.5654, time=1, value=0}
         ]}}})
    )) {module=ScaleMeshSize001}
   &SolveForcesAndVelocity(Clamp Velocity=false, Limit Acceleration=false)
   /SetVariables(
    Particles.Position=&AddVectorToPosition(
     Position=$Particles.Position
     Vector=&Multiply_Vector(
      A=&VectorFromFloat(Value=0.6)
      B=&SelectVectorFromArray(
       Direct Array Index=&ModuloInt(
        A=&MakeIntFromFloat(Float=&Add_Float(
          A=&Multiply_Float(A=$Engine.Time, B=6)
          B=&MakeFloatFromInt(Int=$Particles.SpawnGroup)
         ))
        B=6
       )
       Vector Selection Array=DataInterface() {propertyValues {FloatData [<1 0 0>, <0 -0.5 -0.5>, <-1 0.5 0.5>, <0.5 2 1>, <0.1 -0.4 0.2>, <0.35 0.7 -0.4>]}}
      )
     )
    )
    Particles.Scale=&Multiply_Vector(
     A=$Particles.Scale
     B=&SelectVectorFromArray(
      Direct Array Index=&ModuloInt(
       A=&MakeIntFromFloat(Float=&Add_Float(
         A=&Multiply_Float(A=$Engine.Time, B=6)
         B=&MakeFloatFromInt(Int=$Particles.SpawnGroup)
        ))
       B=3
      )
      Vector Selection Array=DataInterface() {propertyValues {FloatData [<1 1 1>, <1.05 1 1.05>, <0.95 1.02 0.99>]}}
     )
    )
   ) {isSetParameters=true}
  ]
  rendererClasses [&NiagaraMeshRendererProperties]
  renderers [
   {rendererClass=&NiagaraMeshRendererProperties, rendererIndex=0}
   {rendererClass=&NiagaraMeshRendererProperties, rendererIndex=1}
  ]
 }
 Plus {
  emitterUpdateScript [
   &EmitterState()
   &SpawnPerUnit(Module.Use Max Movement Threshold=true, Module.Use Movement Tolerance=false, Spawn Per Unit=95, Use Spawn Probability=false)
   &NM_ToonShaderParameters(Lighting_Hardness=1, RimLight_Color:LinearColor=<rgba 1 1 1 1>, RimLight_Width=0)
   &NM_OutlineParameters()
  ]
  particleSpawnScript [
   &FN_SetSpawnGroup(Number of Spawn Groups=6)
   &PartitionParticles()
   &InitializeParticle(
    Color=&LerpLinearColors(
     StartColor=&SelectLinearColorFromArray(Array Sampling Mode=::NewEnumerator0, Direct Array Index=?, Color Selection Array=$User.MeshColors, Random Seed=0)
     EndColor=&SelectLinearColorFromArray(Array Sampling Mode=::NewEnumerator0, Direct Array Index=?, Color Selection Array=$User.MeshColors_Variant_01, Random Seed=0)
     LerpFactor=$User.ColorSwitcher
    )
    Lifetime Max=0.5
    Lifetime Min=0.3
    Lifetime Mode=::NewEnumerator1
    Mesh Scale Mode=::NewEnumerator1
    Position Offset Coordinate Space=::Local
    Sprite UV Mode=::NewEnumerator3
    UsePositionOffset=true
    Write Material Random=true
   ) {module=InitializeParticle001}
   &MeshRotationForce(Rotation=&VectorFromFloat(Value=1))
   &InitialMeshOrientation(Use Rotation Vector=false) {module=InitialMeshOrientation001}
   &AddVelocity(Velocity=&RandomRangeVector(Minimum=<-1 -1 -1>, Maximum=<1 1 1>), Velocity Speed Scale=10) {module=AddVelocity002}
   &InheritVelocity(
    Inherited Velocity Amount Scale=&Multiply_VectorByFloat(
     Vector=<1 1 1>
     Float=&Multiply_Float(A=$System.InheritVelocityElements, B=&RandomRangeFloat(Minimum=0.5, Maximum=1))
    )
    Inherited Velocity Speed Limit=50
    Speed Limit=true
    UseSpeedThreshold=false
   )
   &ApplyInitialForces()
   &CameraOffset(Camera Offset Amount=25)
   /SetVariables(Particles.Color=&MultiplyLinearColor(
     Linear Color=$Particles.Color
     Scale Factor=&MakeFloat4FromVectorAndFloat("Vector (RGB)"=&VectorFromFloat(Value=$System.TOD), "Float (Alpha)"=1)
    )) {isSetParameters=true}
  ]
  particleUpdateScript [
   &ParticleState()
   &Drag(Drag=&RandomRangeFloat(Minimum=4, Maximum=12), Use Linear Drag=true, Use Rotational Drag=false) {module=Drag001}
   &GravityForce(Gravity=&RandomRangeVector(Minimum=<0 0 150>, Maximum=<0 0 -150>))
   &CurlNoiseForce(Pan Noise=false)
   &ScaleMeshSize(Scale Factor=&Multiply_VectorByFloat(
     Vector=<1 1 1>
     Float=&FloatFromCurve(FloatCurve=DataInterface() {propertyValues {Curve {keys [
          {interpMode=RCIM_Cubic, tangentMode=RCTM_User, time=0, value=0}
          <at 0.2 1>
          {interpMode=RCIM_Cubic, tangentMode=RCTM_User, time=0.7, value=1}
          {arriveTangent=-0.9819, interpMode=RCIM_Cubic, leaveTangent=-0.9819, tangentMode=RCTM_User, time=0.85, value=1.3}
          {arriveTangent=-9.5654, interpMode=RCIM_Cubic, leaveTangent=-9.5654, tangentMode=RCTM_User, time=1, value=0}
         ]}}})
    )) {module=ScaleMeshSize001}
   &SolveForcesAndVelocity(Clamp Velocity=false, Limit Acceleration=false)
   /SetVariables(
    Particles.Position=&AddVectorToPosition(
     Position=$Particles.Position
     Vector=&Multiply_Vector(
      A=&VectorFromFloat(Value=0.6)
      B=&SelectVectorFromArray(
       Direct Array Index=&ModuloInt(
        A=&MakeIntFromFloat(Float=&Add_Float(
          A=&Multiply_Float(A=$Engine.Time, B=6)
          B=&MakeFloatFromInt(Int=$Particles.SpawnGroup)
         ))
        B=6
       )
       Vector Selection Array=DataInterface() {propertyValues {FloatData [<1 0 0>, <0 -0.5 -0.5>, <-1 0.5 0.5>, <0.5 2 1>, <0.1 -0.4 0.2>, <0.35 0.7 -0.4>]}}
      )
     )
    )
    Particles.Scale=&Multiply_Vector(
     A=$Particles.Scale
     B=&SelectVectorFromArray(
      Direct Array Index=&ModuloInt(
       A=&MakeIntFromFloat(Float=&Add_Float(
         A=&Multiply_Float(A=$Engine.Time, B=6)
         B=&MakeFloatFromInt(Int=$Particles.SpawnGroup)
        ))
       B=3
      )
      Vector Selection Array=DataInterface() {propertyValues {FloatData [<1 1 1>, <1.05 1 1.05>, <0.95 1.02 0.99>]}}
     )
    )
   ) {isSetParameters=true}
  ]
  rendererClasses [&NiagaraMeshRendererProperties]
  renderers [
   {rendererClass=&NiagaraMeshRendererProperties, rendererIndex=0}
   {rendererClass=&NiagaraMeshRendererProperties, rendererIndex=1}
  ]
 }
 Shapes_Around_Pickaxe {
  emitterUpdateScript [
   &EmitterState()
   &SpawnBurst_Instantaneous(Spawn Count=2, Use Spawn Probability=false, UseLoopCountLimit=false)
  ]
  particleSpawnScript [
   &PartitionParticles()
   &InitializeParticle(
    Color=&LerpLinearColors(
     StartColor=&SelectLinearColorFromArray(
      Array Sampling Mode=::NewEnumerator1
      Direct Array Index=$Particles.PartitionParticles.Partition
      Color Selection Array=DataInterface() {propertyValues {ColorData [<rgba 0.0482 0.3419 0.855 1>, <rgba 0.855 0.117 0.3185 1>]}}
      Random Seed=?
     )
     EndColor=&SelectLinearColorFromArray(
      Array Sampling Mode=::NewEnumerator1
      Direct Array Index=$Particles.PartitionParticles.Partition
      Color Selection Array=DataInterface() {propertyValues {ColorData [<rgba 0.0185 0.0176 0.0222 1>, <rgba 0.0646 0.1201 0.6198 1>]}}
      Random Seed=?
     )
     LerpFactor=$User.ColorSwitcher
    )
    Lifetime=0.4
    Position Offset=&SelectVectorFromArray(
     Direct Array Index=$Particles.PartitionParticles.Partition
     Vector Selection Array=DataInterface() {propertyValues {FloatData [<-13 0 -2>, <-13 0 -1>]}}
    )
    Sprite Size=&SelectVector2DFromArray(
     Direct Array Index=$Particles.PartitionParticles.Partition
     Vector 2D Selection Array=DataInterface() {propertyValues {FloatData [<55 45>, <55 43>]}}
    )
    Sprite Size Mode=::NewEnumerator4
    UsePositionOffset=true
    Write Material Random=true
   )
   &SpriteFacingAndAlignment(Sprite Alignment=<1 0 0>, Sprite Facing=<0 1 0>, WriteAlignment=true, WriteFacing=true)
   /SetVariables() {isSetParameters=true}
   &ApplyInitialForces()
  ]
  particleUpdateScript [
   &ParticleState()
   &ScaleColor(
    Scale Alpha=&FloatFromCurve(FloatCurve=DataInterface() {propertyValues {Curve {keys [
         <at 0 0 RCIM_Cubic>
         {arriveTangent=-0.0557, interpMode=RCIM_Cubic, leaveTangent=-0.0557, tangentMode=RCTM_User, time=0.2, value=1}
         {arriveTangent=-0.0313, interpMode=RCIM_Cubic, leaveTangent=-0.0313, tangentMode=RCTM_User, time=0.5, value=1}
         <at 0.8 0 RCIM_Cubic>
        ]}}})
    Scale RGB=&VectorFromFloat(Value=$System.TOD)
    ScaleA=true
    ScaleRGB=true
   )
   &DynamicMaterialParameters(
    Index 0 Param 1=&FloatFromCurve(FloatCurve=DataInterface() {propertyValues {Curve {keys [<at 0.5 1 RCIM_Cubic>,<at 1 0 RCIM_Cubic>]}}})
    Module.Param0WriteEnabled=true
    Module.Param1WriteEnabled=false
    Module.Param2WriteEnabled=false
    Module.Param3WriteEnabled=false
   )
  ]
  rendererClasses [&NiagaraSpriteRendererProperties]
  renderers [{rendererClass=&NiagaraSpriteRendererProperties}]
 }
 Stars {
  emitterUpdateScript [
   &EmitterState()
   &SpawnPerUnit(Module.Use Max Movement Threshold=true, Module.Use Movement Tolerance=false, Spawn Per Unit=54, Use Spawn Probability=false)
   &NM_ToonShaderParameters(Lighting_Hardness=0, RimLight_Color:LinearColor=<rgba 1 1 1 1>, RimLight_Width=0)
   &NM_OutlineParameters(OutlineWidth=0.7)
  ]
  particleSpawnScript [
   &PartitionParticles()
   &FN_SetSpawnGroup(Number of Spawn Groups=6)
   &InitializeParticle(
    Color=&LerpLinearColors(
     StartColor=&SelectLinearColorFromArray(Array Sampling Mode=::NewEnumerator0, Direct Array Index=?, Color Selection Array=$User.MeshColors, Random Seed=0)
     EndColor=&SelectLinearColorFromArray(Array Sampling Mode=::NewEnumerator0, Direct Array Index=?, Color Selection Array=$User.MeshColors_Variant_01, Random Seed=0)
     LerpFactor=$User.ColorSwitcher
    )
    Lifetime Max=0.5
    Lifetime Min=0.3
    Lifetime Mode=::NewEnumerator1
    Mesh Scale Mode=::NewEnumerator1
    Mesh Uniform Scale Max=2.5
    Mesh Uniform Scale Min=1.5
    Position Offset Coordinate Space=::Local
    Sprite UV Mode=::NewEnumerator3
    UsePositionOffset=true
    Write Material Random=true
   )
   &MeshRotationForce(Rotation=&VectorFromFloat(Value=1))
   &InitialMeshOrientation(Use Rotation Vector=false) {module=InitialMeshOrientation001}
   &AddVelocity(Velocity=&RandomRangeVector(Minimum=<-1 -1 -1>, Maximum=<1 1 1>), Velocity Speed Scale=10) {module=AddVelocity002}
   &InheritVelocity(
    Inherited Velocity Amount Scale=&Multiply_VectorByFloat(
     Vector=<1 1 1>
     Float=&Multiply_Float(A=$System.InheritVelocityElements, B=&RandomRangeFloat(Minimum=0.5, Maximum=1))
    )
    Inherited Velocity Speed Limit=50
    Speed Limit=true
    UseSpeedThreshold=false
   )
   &ApplyInitialForces()
   &CameraOffset(Camera Offset Amount=25)
   /SetVariables(Particles.Color=&MultiplyLinearColor(
     Linear Color=$Particles.Color
     Scale Factor=&MakeFloat4FromVectorAndFloat("Vector (RGB)"=&VectorFromFloat(Value=$System.TOD), "Float (Alpha)"=1)
    )) {isSetParameters=true}
  ]
  particleUpdateScript [
   &ParticleState()
   &Drag(Drag=&RandomRangeFloat(Minimum=4, Maximum=12), Use Linear Drag=true, Use Rotational Drag=false) {module=Drag001}
   &GravityForce(Gravity=&RandomRangeVector(Minimum=<0 0 150>, Maximum=<0 0 -150>))
   &CurlNoiseForce(Pan Noise=false)
   &ScaleMeshSize(Scale Factor=&Multiply_VectorByFloat(
     Vector=<1 1 1>
     Float=&FloatFromCurve(FloatCurve=DataInterface() {propertyValues {Curve {keys [
          {interpMode=RCIM_Cubic, tangentMode=RCTM_User, time=0, value=0}
          <at 0.2 1>
          {interpMode=RCIM_Cubic, tangentMode=RCTM_User, time=0.7, value=1}
          {arriveTangent=-0.9819, interpMode=RCIM_Cubic, leaveTangent=-0.9819, tangentMode=RCTM_User, time=0.85, value=1.3}
          {arriveTangent=-9.5654, interpMode=RCIM_Cubic, leaveTangent=-9.5654, tangentMode=RCTM_User, time=1, value=0}
         ]}}})
    )) {module=ScaleMeshSize001}
   &SolveForcesAndVelocity(Clamp Velocity=false, Limit Acceleration=false)
   /SetVariables(
    Particles.Position=&AddVectorToPosition(
     Position=$Particles.Position
     Vector=&Multiply_Vector(
      A=&VectorFromFloat(Value=0.6)
      B=&SelectVectorFromArray(
       Direct Array Index=&ModuloInt(
        A=&MakeIntFromFloat(Float=&Add_Float(
          A=&Multiply_Float(A=$Engine.Time, B=6)
          B=&MakeFloatFromInt(Int=$Particles.SpawnGroup)
         ))
        B=6
       )
       Vector Selection Array=DataInterface() {propertyValues {FloatData [<1 0 0>, <0 -0.5 -0.5>, <-1 0.5 0.5>, <0.5 2 1>, <0.1 -0.4 0.2>, <0.35 0.7 -0.4>]}}
      )
     )
    )
    Particles.Scale=&Multiply_Vector(
     A=$Particles.Scale
     B=&SelectVectorFromArray(
      Direct Array Index=&ModuloInt(
       A=&MakeIntFromFloat(Float=&Add_Float(
         A=&Multiply_Float(A=$Engine.Time, B=6)
         B=&MakeFloatFromInt(Int=$Particles.SpawnGroup)
        ))
       B=3
      )
      Vector Selection Array=DataInterface() {propertyValues {FloatData [<1 1 1>, <1.05 1 1.05>, <0.95 1.02 0.99>]}}
     )
    )
   ) {isSetParameters=true}
  ]
  rendererClasses [&NiagaraMeshRendererProperties]
  renderers [
   {rendererClass=&NiagaraMeshRendererProperties, rendererIndex=0}
   {rendererClass=&NiagaraMeshRendererProperties, rendererIndex=1}
  ]
 }
 Trail_Blue {
  emitterUpdateScript [
   &EmitterState(Life Cycle Mode=::NewEnumerator1, Loop Behavior=::NewEnumerator1, Loop Duration=0.5, UseLoopDelay=false)
   &FN_SpawnAnimTrail()
  ]
  particleSpawnScript [
   &FN_InitializeAnimTrailBoneIndecies(Second Bone or Socket Index=0)
   &FN_InitializeAnimTrail(
    AnimTrailWidth=35
    Color=&LerpLinearColors(StartColor=<rgba 0.0482 0.3419 0.855 1>, EndColor=<rgba 0.0185 0.0176 0.0222 1>, LerpFactor=$User.ColorSwitcher)
    Lifetime=0.3
   )
   &DynamicMaterialParameters(
    Index 0 Param 1=&RandomRangeFloat(Minimum=0, Maximum=1)
    Index 0 Param 2=4
    Index 0 Param 3=0.5
    Index 0 Param 4=0.7
    Index 1 Param 2=0.2
    Index 1 Param 3=0.8
    Module.Param0WriteEnabled=true
    Module.Param0WriteEnabled1=true
    Module.Param1WriteEnabled=true
    Module.Param1WriteEnabled1=true
    Module.Param2WriteEnabled=true
    Module.Param2WriteEnabled1=true
    Module.Param3WriteEnabled=true
    Module.Param3WriteEnabled1=false
    Write Parameter Index 1=true
   )
   &ApplyInitialForces()
  ]
  particleUpdateScript [
   &Drag(Drag=5, Use Linear Drag=true, Use Rotational Drag=true)
   &CurlNoiseForce(Noise Frequency=10, Noise Strength=500, Pan Noise=false, Random Seed=66, Randomization Vector:Vector3f=<0.22 0.65 0.23>)
   &SolveForcesAndVelocity(Clamp Velocity=false, Limit Acceleration=false)
   &FN_UpdateAnimTrail(Locked Particle RibbonWidth=1, SkeletalMesh=$User.SkeletalMesh) {module=FN_UpdateAnimTrail001}
   &ScaleColor(
    Scale Alpha=&FloatFromCurve(FloatCurve=DataInterface() {propertyValues {Curve {keys [<at 0.5 1>,<at 0.8 0>]}}})
    Scale RGB=&VectorFromFloat(Value=$System.TOD)
    ScaleA=true
    ScaleRGB=true
   )
   &CameraOffset(Camera Offset Amount=9)
   &ScaleRibbonWidth(Ribbon Width Scale=&FloatFromCurve(FloatCurve=DataInterface() {propertyValues {Curve {keys [<at 0 1>,<at 1 0>]}}}))
   &DynamicMaterialParameters(
    Index 1 Param 1=&FloatFromCurve(FloatCurve=DataInterface() {propertyValues {Curve {keys [
         {interpMode=RCIM_Cubic, tangentMode=RCTM_User, time=0, value=0}
         {arriveTangent=-0.0857, interpMode=RCIM_Cubic, leaveTangent=-0.0857, tangentMode=RCTM_User, time=0.65, value=0.2}
        ]}}})
    Index 1 Param 4=&FloatFromCurve(FloatCurve=DataInterface() {propertyValues {Curve {keys [<at 0.4 1 RCIM_Cubic>,<at 0.9 0 RCIM_Cubic>]}}})
    Module.Param0WriteEnabled1=false
    Module.Param1WriteEnabled1=false
    Module.Param2WriteEnabled1=false
    Module.Param3WriteEnabled1=true
    Write Parameter Index 0=false
    Write Parameter Index 1=true
   ) {module=DynamicMaterialParameters001}
  ]
  rendererClasses [&NiagaraRibbonRendererProperties]
  renderers [{rendererClass=&NiagaraRibbonRendererProperties}]
 }
 Trail_Pink {
  emitterUpdateScript [
   &EmitterState(Life Cycle Mode=::NewEnumerator1, Loop Behavior=::NewEnumerator1, Loop Duration=0.25, UseLoopDelay=false)
   &FN_SpawnAnimTrail()
  ]
  particleSpawnScript [
   &FN_InitializeAnimTrailBoneIndecies(Second Bone or Socket Index=0)
   &FN_InitializeAnimTrail(
    AnimTrailWidth=30
    Color=&LerpLinearColors(StartColor=<rgba 0.855 0.117 0.3185 1>, EndColor=<rgba 0.0646 0.1201 0.6198 1>, LerpFactor=$User.ColorSwitcher)
    Lifetime=0.3
   )
   &DynamicMaterialParameters(
    Index 0 Param 1=&RandomRangeFloat(Minimum=0, Maximum=1)
    Index 0 Param 2=2.5
    Index 0 Param 4=0.7
    Index 1 Param 2=0.2
    Index 1 Param 3=1
    Module.Param0WriteEnabled=true
    Module.Param0WriteEnabled1=true
    Module.Param1WriteEnabled=true
    Module.Param1WriteEnabled1=true
    Module.Param2WriteEnabled=true
    Module.Param2WriteEnabled1=true
    Module.Param3WriteEnabled=true
    Module.Param3WriteEnabled1=false
    Write Parameter Index 1=true
   )
   &ApplyInitialForces()
  ]
  particleUpdateScript [
   &Drag(Drag=5, Use Linear Drag=true, Use Rotational Drag=true)
   &CurlNoiseForce(Noise Frequency=10, Noise Strength=500, Pan Noise=false, Random Seed=33)
   &SolveForcesAndVelocity(Clamp Velocity=false, Limit Acceleration=false)
   &FN_UpdateAnimTrail(Locked Particle RibbonWidth=1, SkeletalMesh=$User.SkeletalMesh) {module=FN_UpdateAnimTrail001}
   &ScaleColor(
    Scale Alpha=&FloatFromCurve(FloatCurve=DataInterface() {propertyValues {Curve {keys [<at 0.5 1>,<at 0.8 0>]}}})
    Scale RGB=&VectorFromFloat(Value=$System.TOD)
    ScaleA=true
    ScaleRGB=true
   )
   &CameraOffset(Camera Offset Amount=10)
   &ScaleRibbonWidth(Ribbon Width Scale=&FloatFromCurve(FloatCurve=DataInterface() {propertyValues {Curve {keys [<at 0 1>,<at 1 0.5>]}}}))
   &DynamicMaterialParameters(
    Index 0 Param 4=&FloatFromCurve(FloatCurve=DataInterface() {propertyValues {Curve {keys [
         {interpMode=RCIM_Cubic, tangentMode=RCTM_User, time=0, value=1}
         {arriveTangent=-0.0857, interpMode=RCIM_Cubic, leaveTangent=-0.0857, tangentMode=RCTM_User, time=0.65, value=0}
        ]}}})
    Index 1 Param 1=&FloatFromCurve(FloatCurve=DataInterface() {propertyValues {Curve {keys [
         {interpMode=RCIM_Cubic, tangentMode=RCTM_User, time=0, value=0}
         {arriveTangent=-0.0857, interpMode=RCIM_Cubic, leaveTangent=-0.0857, tangentMode=RCTM_User, time=0.65, value=0.2}
        ]}}})
    Index 1 Param 4=&FloatFromCurve(FloatCurve=DataInterface() {propertyValues {Curve {keys [<at 0.4 1 RCIM_Cubic>,<at 0.9 0 RCIM_Cubic>]}}})
    Module.Param0WriteEnabled=false
    Module.Param0WriteEnabled1=false
    Module.Param1WriteEnabled=false
    Module.Param1WriteEnabled1=false
    Module.Param2WriteEnabled=false
    Module.Param2WriteEnabled1=false
    Module.Param3WriteEnabled=false
    Module.Param3WriteEnabled1=true
    Write Parameter Index 1=true
   ) {module=DynamicMaterialParameters001}
  ]
  rendererClasses [&NiagaraRibbonRendererProperties]
  renderers [{rendererClass=&NiagaraRibbonRendererProperties}]
 }
 Zaps {
  emitterUpdateScript [
   &EmitterState()
   &SpawnPerUnit(Module.Use Max Movement Threshold=true, Module.Use Movement Tolerance=false, Spawn Per Unit=95, Use Spawn Probability=false)
   &NM_ToonShaderParameters(Lighting_Hardness=1, RimLight_Color:LinearColor=<rgba 1 1 1 1>, RimLight_Width=0)
   &NM_OutlineParameters()
  ]
  particleSpawnScript [
   &PartitionParticles()
   &FN_SetSpawnGroup()
   &InitializeParticle(
    Color=&LerpLinearColors(
     StartColor=&SelectLinearColorFromArray(Array Sampling Mode=::NewEnumerator0, Direct Array Index=?, Color Selection Array=$User.MeshColors, Random Seed=0)
     EndColor=&SelectLinearColorFromArray(Array Sampling Mode=::NewEnumerator0, Direct Array Index=?, Color Selection Array=$User.MeshColors_Variant_01, Random Seed=0)
     LerpFactor=$User.ColorSwitcher
    )
    Lifetime Max=0.5
    Lifetime Min=0.25
    Lifetime Mode=::NewEnumerator1
    Mesh Scale Mode=::NewEnumerator1
    Mesh Uniform Scale Max=3
    Mesh Uniform Scale Min=2
    Position Offset Coordinate Space=::Local
    Sprite UV Mode=::NewEnumerator3
    UsePositionOffset=true
    Write Material Random=true
   )
   &MeshRotationForce(Rotation=&VectorFromFloat(Value=2))
   &InitialMeshOrientation(Use Rotation Vector=false) {module=InitialMeshOrientation001}
   &AddVelocity(Velocity=&RandomRangeVector(Minimum=<-1 -1 -1>, Maximum=<1 1 1>), Velocity Speed Scale=10) {module=AddVelocity002}
   &InheritVelocity(
    Inherited Velocity Amount Scale=&Multiply_VectorByFloat(
     Vector=<1 1 1>
     Float=&Multiply_Float(A=$System.InheritVelocityElements, B=&RandomRangeFloat(Minimum=0.5, Maximum=1))
    )
    Inherited Velocity Speed Limit=50
    Speed Limit=true
    UseSpeedThreshold=false
   )
   &ApplyInitialForces()
   &CameraOffset(Camera Offset Amount=25)
   /SetVariables(Particles.CustomMeshIndex:NiagaraInt32=&RandomRangeInt(Minimum=0, Maximum=1)) {isSetParameters=true}
   /SetVariables(Particles.Color=&MultiplyLinearColor(
     Linear Color=$Particles.Color
     Scale Factor=&MakeFloat4FromVectorAndFloat("Vector (RGB)"=&VectorFromFloat(Value=$System.TOD), "Float (Alpha)"=1)
    )) {isSetParameters=true}
  ]
  particleUpdateScript [
   &ParticleState()
   &Drag(Drag=&RandomRangeFloat(Minimum=4, Maximum=12), Use Linear Drag=true, Use Rotational Drag=false)
   &GravityForce(Gravity=&RandomRangeVector(Minimum=<0 0 150>, Maximum=<0 0 -150>)) {module=GravityForce001}
   &CurlNoiseForce(Pan Noise=false) {module=CurlNoiseForce001}
   &ScaleMeshSize(Scale Factor=&Multiply_VectorByFloat(
     Vector=<1 1 1>
     Float=&FloatFromCurve(FloatCurve=DataInterface() {propertyValues {Curve {keys [
          {interpMode=RCIM_Cubic, tangentMode=RCTM_User, time=0, value=0}
          <at 0.2 1>
          {interpMode=RCIM_Cubic, tangentMode=RCTM_Break, time=0.7, value=1}
          {arriveTangent=0.1352, interpMode=RCIM_Cubic, leaveTangent=0.1352, tangentMode=RCTM_User, time=0.85, value=1}
          {interpMode=RCIM_Cubic, leaveTangent=-9.5654, time=1, value=0}
         ]}}})
    )) {module=ScaleMeshSize001}
   &SolveForcesAndVelocity(Clamp Velocity=false, Limit Acceleration=false)
   /SetVariables(
    Particles.Position=&AddVectorToPosition(
     Position=$Particles.Position
     Vector=&Multiply_Vector(
      A=&VectorFromFloat(Value=0.5)
      B=&SelectVectorFromArray(
       Direct Array Index=&ModuloInt(
        A=&MakeIntFromFloat(Float=&Add_Float(
          A=&Multiply_Float(A=$Engine.Time, B=6)
          B=&MakeFloatFromInt(Int=$Particles.SpawnGroup)
         ))
        B=6
       )
       Vector Selection Array=DataInterface() {propertyValues {FloatData [<1 0 0>, <0 -0.5 -0.5>, <-1 0.5 0.5>, <0.5 2 1>, <0.1 -0.4 0.2>, <0.35 0.7 -0.4>]}}
      )
     )
    )
    Particles.Scale=&Multiply_Vector(
     A=$Particles.Scale
     B=&SelectVectorFromArray(
      Direct Array Index=&ModuloInt(
       A=&MakeIntFromFloat(Float=&Add_Float(
         A=&Multiply_Float(A=$Engine.Time, B=6)
         B=&MakeFloatFromInt(Int=$Particles.SpawnGroup)
        ))
       B=3
      )
      Vector Selection Array=DataInterface() {propertyValues {FloatData [<1 1 1>, <1.05 1 1.05>, <0.95 1.02 0.99>]}}
     )
    )
   ) {isSetParameters=true}
  ]
  rendererClasses [&NiagaraMeshRendererProperties]
  renderers [
   {rendererClass=&NiagaraMeshRendererProperties, rendererIndex=0}
   {rendererClass=&NiagaraMeshRendererProperties, rendererIndex=1}
  ]
 }
}
summary {
 systemName=NS_Pickaxe_Dazzle_Trail
 emitters [
  {emitterName=Trail_Pink, rendererClasses [&NiagaraRibbonRendererProperties]}
  {emitterName=Trail_Blue, rendererClasses [&NiagaraRibbonRendererProperties]}
  {emitterName=Zaps, rendererClasses [&NiagaraMeshRendererProperties]}
  {emitterName=Stars, rendererClasses [&NiagaraMeshRendererProperties]}
  {emitterName=Plus, rendererClasses [&NiagaraMeshRendererProperties]}
  {emitterName=Bunny001, rendererClasses [&NiagaraMeshRendererProperties]}
  {emitterName=Emote, rendererClasses [&NiagaraMeshRendererProperties]}
  {emitterName=Shapes_Around_Pickaxe, rendererClasses [&NiagaraSpriteRendererProperties]}
 ]
 userVariables [
  User.bIsFrontEndPreview {type=NiagaraBool}
  User.ColorSwitcher {type=NiagaraFloat, defaultValue=0, description=""}
  User.MeshColors {
   type=NiagaraDataInterfaceArrayColor
   description=""
   defaultValue {
    struct=NiagaraExt_VariableValue_DataInterface
    value {
     dataInterface=&self:NiagaraDataInterfaceArrayColor_0
     dataInterfaceClass=&NiagaraDataInterfaceArrayColor
    }
   }
  }
  User.MeshColors_Variant_01 {
   type=NiagaraDataInterfaceArrayColor
   description=""
   defaultValue {
    struct=NiagaraExt_VariableValue_DataInterface
    value {
     dataInterface=&self:NiagaraDataInterfaceArrayColor_1
     dataInterfaceClass=&NiagaraDataInterfaceArrayColor
    }
   }
  }
  User.ParaGlideLeanAlpha {type=NiagaraFloat, defaultValue=0, description=""}
  User.PlayerSpeed {type=NiagaraFloat, defaultValue=0, description=""}
  User.RibbonIDOffset {type=NiagaraInt32}
  User.SkeletalMesh {
   type=NiagaraDataInterfaceSkeletalMesh
   defaultValue {
    struct=NiagaraExt_VariableValue_DataInterface
    value {dataInterface=&self:NiagaraDataInterfaceSkeletalMesh_0}
   }
  }
 ]
}
